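# Control Rig for the SeleneDonna female medium body that corrects the jacket's neck flap bones as the character moves.
# It runs a shared DeformRig function, then parent-constrains the flap bones to weighted head, neck, clavicle, arm and spine bones.
# Corrective transforms are layered on, weighted by head rotation deltas, neck swing and spherical pose readers on both upper arms.
# A pelvis-height curve (CrouchWalk) scales an extra head-up correction.
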ControlRigBlueprint /BRCosmetics/Characters/Player/Female/Medium/Bodies/F_MED_SeleneDonna/Meshes/F_MED_SeleneDonna_CtrlRig
graphs [
 DeltaRotation_ContainedGraph {nodes [
   n1 = RigVMFunctionEntryNode()
   RigVMFunction_MathBoolToInteger = &RigVMFunction_MathBoolToInteger(Value=@n1.onlyTwist)
   RigVMFunction_MathQuaternionSwingTwist = &RigVMFunction_MathQuaternionSwingTwist(
    twistAxis=@n1.twistAxis
    Input=&RigUnit_GetRelativeTransformForItem(
     bChildInitial=False
     bParentInitial=False
     child=(Type=Bone,Name="None")
     parent=(Type=Bone,Name="None")
     child.Name=@n1.child
     parent.Name=@n1.parent
     parent.Type=@n1.parentType
    ).RelativeTransform.Rotation
   )
   RigVMFunction_MathQuaternionToEuler = &RigVMFunction_MathQuaternionToEuler(
    Value=RigVMDispatchNode(
     Values=((),())
     Index=@RigVMFunction_MathBoolToInteger.Result
     Values.0=@RigVMFunction_MathQuaternionSwingTwist.Swing
     Values.1=@RigVMFunction_MathQuaternionSwingTwist.Twist
    ) {template="DISPATCH_RigVMDispatch_SelectInt32(in Index,in Values,out Result)"}.Result
    RotationOrder=@n1.Rotation Order
   )
   > DISPATCH_RigVMDispatch_Print = RigVMDispatchNode(ScreenColor=(R=1,G=1,B=1,A=1), ScreenDuration=0.1, Enabled=@n1.debugRotation, Value=@RigVMFunction_MathQuaternionToEuler.Result) {template="DISPATCH_RigVMDispatch_Print(in Prefix,in Value,in Enabled,in ScreenDuration,in ScreenColor)"}
   Clamp = &RigVMFunction_MathIntClamp(Value=@n1.primAxis)
   DISPATCH_RigVMDispatch_SelectInt32 = RigVMDispatchNode(
    Values=(0,0,0)
    Index=@Clamp.Result
    Values.0=@RigVMFunction_MathQuaternionToEuler.Result.X
    Values.1=@RigVMFunction_MathQuaternionToEuler.Result.Y
    Values.2=@RigVMFunction_MathQuaternionToEuler.Result.Z
   ) {template="DISPATCH_RigVMDispatch_SelectInt32(in Index,in Values,out Result)"}
   Remap = &RigVMFunction_MathDoubleRemap(
    TargetMaximum=1
    TargetMinimum=0
    bClamp=False
    Value=@DISPATCH_RigVMDispatch_SelectInt32.Result
    SourceMaximum=@n1.max
    SourceMinimum=@n1.min
   )
   Multiply = &RigVMFunction_MathDoubleMul(A=&RigVMFunction_MathDoubleClamp(Value=@Remap.Result).Result, B=@n1.premult)
   > RigVMFunctionReturnNode(outputWeight=@Multiply.Result)
  ]}
 DeltaRotation_ContainedGraph {nodes [
   n1 = RigVMFunctionEntryNode()
   RigVMFunction_MathBoolToInteger = &RigVMFunction_MathBoolToInteger(Value=@n1.onlyTwist)
   RigVMFunction_MathQuaternionSwingTwist = &RigVMFunction_MathQuaternionSwingTwist(
    twistAxis=@n1.twistAxis
    Input=&RigUnit_GetRelativeTransformForItem(
     bChildInitial=False
     bParentInitial=False
     child=(Type=Bone,Name="None")
     parent=(Type=Bone,Name="None")
     child.Name=@n1.child
     parent.Name=@n1.parent
     parent.Type=@n1.parentType
    ).RelativeTransform.Rotation
   )
   RigVMFunction_MathQuaternionToEuler = &RigVMFunction_MathQuaternionToEuler(
    Value=RigVMDispatchNode(
     Values=((),())
     Index=@RigVMFunction_MathBoolToInteger.Result
     Values.0=@RigVMFunction_MathQuaternionSwingTwist.Swing
     Values.1=@RigVMFunction_MathQuaternionSwingTwist.Twist
    ) {template="DISPATCH_RigVMDispatch_SelectInt32(in Index,in Values,out Result)"}.Result
    RotationOrder=@n1.Rotation Order
   )
   > DISPATCH_RigVMDispatch_Print = RigVMDispatchNode(ScreenColor=(R=1,G=1,B=1,A=1), ScreenDuration=0.1, Enabled=@n1.debugRotation, Value=@RigVMFunction_MathQuaternionToEuler.Result) {template="DISPATCH_RigVMDispatch_Print(in Prefix,in Value,in Enabled,in ScreenDuration,in ScreenColor)"}
   Clamp = &RigVMFunction_MathIntClamp(Value=@n1.primAxis)
   DISPATCH_RigVMDispatch_SelectInt32 = RigVMDispatchNode(
    Values=(0,0,0)
    Index=@Clamp.Result
    Values.0=@RigVMFunction_MathQuaternionToEuler.Result.X
    Values.1=@RigVMFunction_MathQuaternionToEuler.Result.Y
    Values.2=@RigVMFunction_MathQuaternionToEuler.Result.Z
   ) {template="DISPATCH_RigVMDispatch_SelectInt32(in Index,in Values,out Result)"}
   Remap = &RigVMFunction_MathDoubleRemap(
    TargetMaximum=1
    TargetMinimum=0
    bClamp=False
    Value=@DISPATCH_RigVMDispatch_SelectInt32.Result
    SourceMaximum=@n1.max
    SourceMinimum=@n1.min
   )
   Multiply = &RigVMFunction_MathDoubleMul(A=&RigVMFunction_MathDoubleClamp(Value=@Remap.Result).Result, B=@n1.premult)
   > RigVMFunctionReturnNode(outputWeight=@Multiply.Result)
  ]}
 New Function_ContainedGraph {nodes [
   n1 = RigVMFunctionEntryNode()
   RigVMFunction_MathBoolToInteger = &RigVMFunction_MathBoolToInteger(Value=@n1.onlyTwist)
   RigVMFunction_MathQuaternionSwingTwist = &RigVMFunction_MathQuaternionSwingTwist(
    twistAxis=@n1.twistAxis
    Input=&RigUnit_GetRelativeTransformForItem(
     bChildInitial=False
     bParentInitial=False
     child=(Type=Bone,Name="None")
     parent=(Type=Bone,Name="None")
     child.Name=@n1.child
     parent.Name=@n1.parent
     parent.Type=@n1.parentType
    ).RelativeTransform.Rotation
   )
   RigVMFunction_MathQuaternionToEuler = &RigVMFunction_MathQuaternionToEuler(
    Value=RigVMDispatchNode(
     Values=((),())
     Index=@RigVMFunction_MathBoolToInteger.Result
     Values.0=@RigVMFunction_MathQuaternionSwingTwist.Swing
     Values.1=@RigVMFunction_MathQuaternionSwingTwist.Twist
    ) {template="DISPATCH_RigVMDispatch_SelectInt32(in Index,in Values,out Result)"}.Result
    RotationOrder=@n1.Rotation Order
   )
   > DISPATCH_RigVMDispatch_Print = RigVMDispatchNode(ScreenColor=(R=1,G=1,B=1,A=1), ScreenDuration=1.5, Enabled=@n1.debugRotation, Value=@RigVMFunction_MathQuaternionToEuler.Result) {template="DISPATCH_RigVMDispatch_Print(in Prefix,in Value,in Enabled,in ScreenDuration,in ScreenColor)"}
   Clamp = &RigVMFunction_MathIntClamp(Value=@n1.primAxis)
   DISPATCH_RigVMDispatch_SelectInt32 = RigVMDispatchNode(
    Values=(0,0,0)
    Index=@Clamp.Result
    Values.0=@RigVMFunction_MathQuaternionToEuler.Result.X
    Values.1=@RigVMFunction_MathQuaternionToEuler.Result.Y
    Values.2=@RigVMFunction_MathQuaternionToEuler.Result.Z
   ) {template="DISPATCH_RigVMDispatch_SelectInt32(in Index,in Values,out Result)"}
   Remap = &RigVMFunction_MathDoubleRemap(
    TargetMaximum=1
    TargetMinimum=0
    bClamp=False
    Value=@DISPATCH_RigVMDispatch_SelectInt32.Result
    SourceMaximum=@n1.max
    SourceMinimum=@n1.min
   )
   Multiply = &RigVMFunction_MathDoubleMul(A=&RigVMFunction_MathDoubleClamp(Value=@Remap.Result).Result, B=@n1.premult)
   > RigVMFunctionReturnNode(outputWeight=@Multiply.Result)
  ]}
 New Function_ContainedGraph {nodes [
   n1 = RigVMFunctionEntryNode()
   RigVMFunction_MathQuaternionToEuler = &RigVMFunction_MathQuaternionToEuler(
    RotationOrder=YXZ
    Value=&RigUnit_GetRelativeTransformForItem(
     Child=(Type=Bone,Name="head")
     Parent=(Type=Bone,Name="neck_01")
     bChildInitial=False
     bParentInitial=False
     Child.Name=@n1.Child
     Parent.Name=@n1.Parent
    ).RelativeTransform.Rotation
   )
   Remap = &RigVMFunction_MathDoubleRemap(
    TargetMaximum=1
    TargetMinimum=0
    SourceMaximum=@n1.Source Max
    SourceMinimum=@n1.Source Min
    Value=@RigVMFunction_MathQuaternionToEuler.Result.Z
   )
   > RigVMFunctionReturnNode(
    Delta_z=@Remap.Result
    Delta_X_Negative=&RigVMFunction_MathDoubleRemap(SourceMaximum=-60, SourceMinimum=0, TargetMaximum=1, TargetMinimum=0, Value=@RigVMFunction_MathQuaternionToEuler.Result.X).Result
    Delta_X_Positive=&RigVMFunction_MathDoubleRemap(SourceMaximum=60, SourceMinimum=0, TargetMaximum=1, TargetMinimum=0, Value=@RigVMFunction_MathQuaternionToEuler.Result.X).Result
   )
  ]}
 New Function_ContainedGraph {nodes [
   n1 = RigVMFunctionEntryNode()
   RigVMFunction_MathQuaternionToEuler = &RigVMFunction_MathQuaternionToEuler(
    RotationOrder=ZXY
    Value=&RigUnit_GetRelativeTransformForItem(
     Child=(Type=Bone,Name="head")
     Parent=(Type=Bone,Name="neck_01")
     bChildInitial=False
     bParentInitial=False
     Child.Name=@n1.Child
     Parent.Name=@n1.Parent
    ).RelativeTransform.Rotation
   )
   Remap = &RigVMFunction_MathDoubleRemap(
    TargetMaximum=1
    TargetMinimum=0
    SourceMaximum=@n1.Source Max
    SourceMinimum=@n1.Source Min
    Value=@RigVMFunction_MathQuaternionToEuler.Result.Z
   )
   > RigVMFunctionReturnNode(
    Delta_z=@Remap.Result
    Delta_Y_Negative=&RigVMFunction_MathDoubleRemap(SourceMaximum=-10, SourceMinimum=0, TargetMaximum=1, TargetMinimum=0, Value=@RigVMFunction_MathQuaternionToEuler.Result.Y).Result
    Delta_Y_Positive=&RigVMFunction_MathDoubleRemap(SourceMaximum=20, SourceMinimum=0, TargetMaximum=1, TargetMinimum=0, Value=@RigVMFunction_MathQuaternionToEuler.Result.Y).Result
   )
  ]}
 RigVMFunctionLibrary {nodes [
   DeltaRotation = RigVMCollapseNode(ExecuteContext=(), Rotation Order=ZYX, premult=1, twistAxis=(X=1,Y=0,Z=0))
   DeltaRotation_1 = RigVMCollapseNode(ExecuteContext=(), Rotation Order=ZYX, debugRotation=false, premult=1, twistAxis=(X=1,Y=0,Z=0))
   DeltaRotation_2 = RigVMCollapseNode(ExecuteContext=(), Rotation Order=ZYX, debugRotation=false, premult=1, twistAxis=(X=1,Y=0,Z=0))
   DeltaRotation_3 = RigVMCollapseNode(ExecuteContext=())
   DeltaRotation_4 = RigVMCollapseNode(ExecuteContext=())
  ]}
 RigVMFunction_Sequence_1_ContainedGraph {nodes [
   n1 = RigVMFunctionEntryNode()
   RigVMFunction_Sequence = &RigVMFunction_Sequence(ExecuteContext=@n1.ExecuteContext)
   n2 = &RigVMFunction_Sequence(ExecuteContext=@RigVMFunction_Sequence.B)
   n3 = &RigVMFunction_Sequence(ExecuteContext=@n2.B)
   n4 = &RigVMFunction_Sequence(ExecuteContext=@n3.B)
   RigVMFunctionReturnNode(A=@RigVMFunction_Sequence.A, B=@n2.A, C=@n3.A, D=@n4.A, E=@n4.B)
  ]}
 RigVMModel {nodes [
   RigUnit_BeginExecution = &RigUnit_BeginExecution()
   RigVMFunction_Sequence = RigVMAggregateNode(ExecuteContext=@RigUnit_BeginExecution.ExecutePin)
   FN_DeformRig_FNC = RigVMFunctionReferenceNode(
    DeformRig_ Curve="(deformRig_l_upperarm_up_alpha_72_#1=False,deformRig_r_upperarm_up_alpha_73_#2=False,deformRig_l_upperarm_fr_alpha_74_#3=true,deformRig_r_upperarm_fr_alpha_75_CB51B80 ...(truncated)"
    DeformRig_Bool="(upperarm_1_#4=True,elbow_3_#5=True,wrist_7_#6=True,groin_8_#7=True,glute_17_#8=True,knee_9_40C408324CA261 ...(truncated)"
    DeformRig_Debug=(upperarm_debug_18_#4=False,wrist_debug_20_#6=False,lat_debug_26_#9=False,groinGlute_debug_27_#7=False)
    DeformRig_Value="(upperarm_up_18_#10=(Rotation=(X=0,Y=0,Z=0,W=1),Translation=(X=0,Y=0,Z=8),Scale3D=(X=0.7,Y=1,Z=1)),upperarm_fwd_fr_19_#11=(Rotation=(X=0,Y=0,Z=0,W=1),Translation=(X=0,Y=10,Z=0),Scale3D=(X=0.5,Y=1,Z= ...(truncated)"
    ExecuteContext=@RigVMFunction_Sequence.B
   )
   &RigUnit_ParentConstraint(
    AdvancedSettings=(RotationOrderForFilter=XZY)
    Child=(Type=Bone,Name="base_jacket_flapcn")
    Parents="((Item=(Type=Bone,Name=\"head\"),Weight=0.0002),(Item=(Type=Bone,Name=\"neck_01\"),Weight=0.0006),(Item=(Type=Bone,Name=\"neck_02\"),Weight=0.0009),(Item=(Type=Bone,Name=\"clavicle_l\"),Weight=0.0162),(Item=(Type=Bone,Name=\"clavicle_r\"),Weight=0.0162),(Item=(Type= ...(truncated)"
    Weight=1
    ExecutePin=@RigVMFunction_Sequence.C
   )
   > &RigUnit_ParentConstraint(
    AdvancedSettings=(RotationOrderForFilter=XZY)
    Child=(Type=Bone,Name="base_jacket_flapneck_01_r")
    Parents="((Item=(Type=Bone,Name=\"neck_01\"),Weight=0.0003),(Item=(Type=Bone,Name=\"neck_02\"),Weight=0.0008),(Item=(Type=Bone,Name=\"clavicle_l\"),Weight=0.0023),(Item=(Type=Bone,Name=\"upperarm_twist_01_r\"),Weight=0.0615),(Item=(Type=Bone,Name=\"spine_04\"),Weight=0.0916) ...(truncated)"
    Weight=1
   )
   > &RigUnit_ParentConstraint(
    AdvancedSettings=(RotationOrderForFilter=XZY)
    Child=(Type=Bone,Name="base_jacket_flapneck_01_l")
    Parents="((Item=(Type=Bone,Name=\"neck_01\"),Weight=0.0005),(Item=(Type=Bone,Name=\"neck_02\"),Weight=0.0011),(Item=(Type=Bone,Name=\"clavicle_r\"),Weight=0.0032),(Item=(Type=Bone,Name=\"upperarm_twist_01_l\"),Weight=0.0743),(Item=(Type=Bone,Name=\"spine_04\"),Weight=0.0927) ...(truncated)"
    Weight=1
   )
   > &RigUnit_ParentConstraint(
    AdvancedSettings=(RotationOrderForFilter=XZY)
    Child=(Type=Bone,Name="base_jacket_flapneck_02_l")
    Parents="((Item=(Type=Bone,Name=\"head\"),Weight=0.0003),(Item=(Type=Bone,Name=\"neck_01\"),Weight=0.0033),(Item=(Type=Bone,Name=\"neck_02\"),Weight=0.0071),(Item=(Type=Bone,Name=\"clavicle_r\"),Weight=0.011),(Item=(Type=Bone,Name=\"spine_04\"),Weight=0.0127),(Item=(Type=Bon ...(truncated)"
    Weight=1
   )
   > &RigUnit_ParentConstraint(
    AdvancedSettings=(RotationOrderForFilter=XZY)
    Child=(Type=Bone,Name="base_jacket_flapneck_02_r")
    Parents="((Item=(Type=Bone,Name=\"head\"),Weight=0.0003),(Item=(Type=Bone,Name=\"neck_01\"),Weight=0.0033),(Item=(Type=Bone,Name=\"neck_02\"),Weight=0.0071),(Item=(Type=Bone,Name=\"clavicle_l\"),Weight=0.011),(Item=(Type=Bone,Name=\"spine_04\"),Weight=0.0127),(Item=(Type=Bon ...(truncated)"
    Weight=1
   )
   > DeltaRotation_6 = RigVMFunctionReferenceNode(
    Rotation Order=ZYX
    child=head
    debugRotation=false
    max=45
    min=28
    parent=head_null
    parentType=Null
    premult=1
    primAxis=2
    twistAxis=(X=1,Y=0,Z=0)
   )
   > &RigUnit_ModifyTransforms(
    ItemToModify="((Item=(Type=Bone,Name=\"base_jacket_flapneck_01_l\"),Transform=(Rotation=(X=0,Y=0.1736,Z=0,W=0.9848),Translation=(X=0,Y=0,Z=0),Scale3D=(X=1,Y=1,Z=1))),(Item=(Type=Bone,Name=\"base_jacket_flapneck_01_r\"),Transform=(Rotation=(X=0,Y=-0.1736,Z=0,W=0.9848),Transl ...(truncated)"
    WeightMinimum=0
    Weight=@DeltaRotation_6.outputWeight
   )
   > RigVMVariableNode(Value=@DeltaRotation_6.outputWeight) {variable=isHeadDownDelta}
   > DeltaRotation_8 = RigVMFunctionReferenceNode(
    Rotation Order=ZYX
    child=head
    debugRotation=false
    max=-45
    min=10
    parent=head_null
    parentType=Null
    premult=1
    primAxis=2
    twistAxis=(X=1,Y=0,Z=0)
   )
   > &RigUnit_ModifyTransforms(
    ItemToModify="((Item=(Type=Bone,Name=\"base_jacket_flapcn\"),Transform=(Rotation=(X=0,Y=0,Z=-0.4226,W=0.9063),Translation=(X=0,Y=0,Z=0),Scale3D=(X=1,Y=1,Z=1))),(Item=(Type=Bone,Name=\"base_jacket_flapneck_02_l\"),Transform=(Rotation=(X=0,Y=0,Z=-0.342,W=0.9397),Translation=( ...(truncated)"
    WeightMinimum=0
    Weight=@DeltaRotation_8.outputWeight
   )
   > RigVMVariableNode(Value=@DeltaRotation_8.outputWeight) {variable=isHeadUpDelta}
   n1 = &RigVMFunction_AnimEvalRichCurve(
    Curve=(EditorCurveData=(Keys=((),(Time=0.5,Value=1),(Time=0.8))))
    SourceMaximum=2
    SourceMinimum=0
    TargetMinimum=0
    Value=&RigVMFunction_MathDoubleRemap(
     SourceMaximum=75
     SourceMinimum=45
     TargetMaximum=1
     TargetMinimum=0
     bClamp=false
     Value=&RigUnit_GetRelativeTransformForItem(Child=(Type=Bone,Name="pelvis"), Parent=(Type=Bone,Name="root"), bChildInitial=False, bParentInitial=False).RelativeTransform.Translation.Z
    ).Result
   )
   > RigVMVariableNode(Value=@n1.Result) {variable=CrouchWalk}
   > n2 = &RigUnit_SphericalPoseReader(
    ActiveRegionScaleFactors=(PositiveWidth=1,NegativeWidth=1,PositiveHeight=1,NegativeHeight=1)
    ActiveRegionSize=0.1
    Debug=(bDrawDebug=false,bDraw2D=False,bDrawLocalAxes=False,DebugScale=25,DebugSegments=20,DebugThickness=0.25)
    DriverAxis=(X=1,Y=0,Z=0)
    DriverItem=(Type=Bone,Name="upperarm_l")
    FalloffRegionScaleFactors=(PositiveWidth=1,NegativeWidth=1,PositiveHeight=1,NegativeHeight=1)
    FalloffSize=0.2
    FlipHeightScaling=False
    FlipWidthScaling=False
    OptionalParentItem=(Type=Bone,Name="None")
    RotationOffset=(X=10,Y=-55,Z=0)
   )
   > &RigUnit_ModifyTransforms(
    ItemToModify="((Item=(Type=Bone,Name=\"base_jacket_flapneck_02_r\"),Transform=(Rotation=(X=0.0048,Y=0.3419,Z=-0.1198,W=-0.9321),Translation=(X=0,Y=0,Z=0),Scale3D=(X=1,Y=1,Z=1))),(Item=(Type=Bone,Name=\"base_jacket_flapneck_01_r\"),Transform=(Rotation=(X=0.4305,Y=0.339,Z=0.2 ...(truncated)"
    WeightMinimum=0
    Weight=@n2.OutputParam
   )
   > &RigUnit_ModifyTransforms(
    ItemToModify="((Item=(Type=Bone,Name=\"base_jacket_flapcn\"),Transform=(Rotation=(X=0.1942,Y=0.1135,Z=0.1829,W=-0.957),Translation=(X=-2.3968,Y=-3.6423,Z=1.0963),Scale3D=(X=0.6611,Y=1,Z=0.9422))),(Item=(Type=Bone,Name=\"base_jacket_flapneck_02_l\"),Transform=(Rotation=(X=0. ...(truncated)"
    WeightMinimum=0
    Weight=&RigVMFunction_MathDoubleMul(A=RigVMVariableNode() {variable=isHeadUpDelta}.Value, B=RigVMVariableNode() {variable=CrouchWalk}.Value).Result
   )
   > DeltaRotation_7 = RigVMFunctionReferenceNode(Child=Head, Parent=neck_01, Source Max=-60, Source Min=0)
   > &RigUnit_ModifyTransforms(
    ItemToModify="((Item=(Type=Bone,Name=\"base_jacket_flapneck_02_l\"),Transform=(Rotation=(X=-0.065,Y=-0.2708,Z=0.1117,W=-0.9539),Translation=(X=0,Y=0,Z=0),Scale3D=(X=1,Y=1,Z=1))),(Item=(Type=Bone,Name=\"base_jacket_flapneck_01_l\"),Transform=(Rotation=(X=-0.1155,Y=-0.5101,Z= ...(truncated)"
    WeightMinimum=0
    Weight=@DeltaRotation_7.Delta_X_Negative
   )
   > &RigUnit_ModifyTransforms(
    ItemToModify="((Item=(Type=Bone,Name=\"base_jacket_flapcn\"),Transform=(Rotation=(X=0.134,Y=0.179,Z=0.1987,W=-0.9542),Translation=(X=0,Y=0,Z=0),Scale3D=(X=1,Y=1,Z=1))),(Item=(Type=Bone,Name=\"base_jacket_flapneck_02_l\"),Transform=(Rotation=(X=-0.0536,Y=-0.0298,Z=0.1205,W=- ...(truncated)"
    WeightMinimum=0
    Weight=@DeltaRotation_7.Delta_X_Positive
   )
   > &RigUnit_ParentConstraint(
    AdvancedSettings=(InterpolationType=Average,RotationOrderForFilter=XZY)
    Child=(Type=Bone,Name="neck_02_helper")
    Parents=((Item=(Type=Bone,Name="neck_02"),Weight=1))
    Weight=1
   )
   RigVMFunction_MathQuaternionSwingTwist = &RigVMFunction_MathQuaternionSwingTwist(
    TwistAxis=(X=1,Y=0,Z=0)
    Input=&RigUnit_GetTransform(Item=(Type=Bone,Name="neck_02_helper"), Space=LocalSpace, bInitial=false).Transform.Rotation
   )
   RigVMFunction_MathQuaternionToEuler = &RigVMFunction_MathQuaternionToEuler(RotationOrder=XZY, Value=@RigVMFunction_MathQuaternionSwingTwist.Swing)
   Remap = &RigVMFunction_MathDoubleRemap(SourceMaximum=-14, SourceMinimum=0, TargetMaximum=1, TargetMinimum=0, Value=@RigVMFunction_MathQuaternionToEuler.Result.Y)
   > &RigUnit_OffsetTransformForItem(
    Item=(Type=Bone,Name="base_jacket_flapneck_02_r")
    OffsetTransform=(Rotation=(X=0.2048,Y=-0.1336,Z=0.0901,W=-0.9654),Translation=(X=-0.1928,Y=-2.1231,Z=2.0388),Scale3D=(X=1,Y=1,Z=1))
    bPropagateToChildren=True
    Weight=@Remap.Result
   )
   > &RigUnit_OffsetTransformForItem(
    Item=(Type=Bone,Name="base_jacket_flapneck_01_r")
    OffsetTransform=(Rotation=(X=0.2675,Y=0.189,Z=0.2186,W=-0.9192),Translation=(X=0,Y=0,Z=0),Scale3D=(X=1,Y=1,Z=1))
    bPropagateToChildren=True
    Weight=@Remap.Result
   )
   n3 = &RigVMFunction_MathDoubleRemap(SourceMaximum=20, SourceMinimum=0, TargetMaximum=1, TargetMinimum=0, Value=@RigVMFunction_MathQuaternionToEuler.Result.Y)
   > &RigUnit_OffsetTransformForItem(
    Item=(Type=Bone,Name="base_jacket_flapneck_02_l")
    OffsetTransform=(Rotation=(X=-0.147,Y=-0.4037,Z=0.2722,W=-0.861),Translation=(X=0.4059,Y=0.7564,Z=-0.9725),Scale3D=(X=1,Y=1,Z=1))
    bPropagateToChildren=True
    Weight=@n3.Result
   )
   > &RigUnit_OffsetTransformForItem(
    Item=(Type=Bone,Name="base_jacket_flapneck_01_l")
    OffsetTransform=(Rotation=(X=-0.344,Y=-0.5362,Z=-0.0244,W=-0.7704),Translation=(X=0,Y=0,Z=0),Scale3D=(X=1,Y=1,Z=1))
    bPropagateToChildren=True
    Weight=@n3.Result
   )
   > n4 = &RigUnit_SphericalPoseReader(
    ActiveRegionScaleFactors=(PositiveWidth=1,NegativeWidth=1,PositiveHeight=1,NegativeHeight=1)
    ActiveRegionSize=0.05
    Debug=(bDrawDebug=false,bDraw2D=False,bDrawLocalAxes=False,DebugScale=25,DebugSegments=20,DebugThickness=0.25)
    DriverAxis=(X=-1,Y=0,Z=0)
    DriverItem=(Type=Bone,Name="upperarm_r")
    FalloffRegionScaleFactors=(PositiveWidth=1,NegativeWidth=1,PositiveHeight=1,NegativeHeight=1)
    FalloffSize=0.15
    FlipHeightScaling=False
    FlipWidthScaling=False
    OptionalParentItem=(Type=Bone,Name="None")
    RotationOffset=(X=-75,Y=0,Z=0)
   )
   > &RigUnit_OffsetTransformForItem(
    Item=(Type=Bone,Name="base_jacket_flapneck_02_r")
    OffsetTransform=(Rotation=(X=0.0617,Y=-0.4029,Z=0.1329,W=0.9034),Translation=(X=1.5611,Y=0.2137,Z=-1.5521),Scale3D=(X=1,Y=1,Z=0.3228))
    bPropagateToChildren=True
    Weight=@n4.OutputParam
   )
   RigUnit_GetTransform = &RigUnit_GetTransform(Item=(Type=Control,Name="neck_01_ctrl"), bInitial=False)
   RigUnit_SetTransform = &RigUnit_SetTransform(ExecutePin=(), Item=(Type=Bone,Name="neck_01"), Weight=1, bInitial=False, Value=@RigUnit_GetTransform.Transform)
   > &RigUnit_SetTransform(
    Item=(Type=Bone,Name="neck_02")
    Weight=1
    bInitial=False
    Value=&RigUnit_GetTransform(Item=(Type=Control,Name="neck_02_ctrl"), bInitial=False).Transform
   )
   > &RigUnit_SetTransform(
    Item=(Type=Bone,Name="head")
    Weight=1
    bInitial=False
    Value=&RigUnit_GetTransform(Item=(Type=Control,Name="head_ctrl"), bInitial=False).Transform
   )
   > &RigUnit_SetTransform(
    Item=(Type=Bone,Name="clavicle_l")
    Weight=1
    bInitial=False
    Value=&RigUnit_GetTransform(Item=(Type=Control,Name="clavicle_l_ctrl"), bInitial=False).Transform
   )
   > &RigUnit_SetTransform(
    Item=(Type=Bone,Name="clavicle_r")
    Weight=1
    bInitial=False
    Value=&RigUnit_GetTransform(Item=(Type=Control,Name="clavicle_r_ctrl"), bInitial=False).Transform
   )
   &RigUnit_ModifyTransforms(
    ExecutePin=()
    ItemToModify="((Item=(Type=Bone,Name=\"base_jacket_flapcn\"),Transform=(Rotation=(X=0.0406,Y=0.1035,Z=0.2977,W=-0.9482),Translation=(X=0.5552,Y=1.2987,Z=0.1229),Scale3D=(X=0.6611,Y=1,Z=1.604))),(Item=(Type=Bone,Name=\"base_jacket_flapneck_02_l\"),Transform=(Rotation=(X=0.10 ...(truncated)"
    Weight=1
    WeightMinimum=0
   )
   &RigUnit_OffsetTransformForItem(
    ExecutePin=()
    Item=(Type=Bone,Name="base_jacket_flapcn")
    OffsetTransform=(Rotation=(X=0.2492,Y=0,Z=0,W=-0.9685),Translation=(X=0,Y=0,Z=0),Scale3D=(X=1,Y=1,Z=1))
    Weight=0
    bPropagateToChildren=True
   )
  ]}
]
hierarchy [
 root {loc=<0 0 0>, rot=<0 0 0 1>}
 pelvis {parent=root, loc=<0 -0.5301 98.0571>, rot=<0.0127 0.707 -0.0127 -0.707>}
 spine_01 {parent=pelvis, loc=<6.1208 -0.9328 0>, rot=<0 0 0.034 -0.9994>}
 spine_02 {parent=spine_01, loc=<6.5997 -0.2095 0>, rot=<0 0 0.0238 -0.9997>}
 spine_03 {parent=spine_02, loc=<6.4914 -0.1237 0>, rot=<0 0 0.0149 -0.9999>}
 spine_04 {parent=spine_03, loc=<6.9598 0.0794 0>, rot=<0 0 -0.0132 -0.9999>}
 spine_05 {parent=spine_04, loc=<6.9987 0 0>, rot=<0 0 -0.0015 -1>}
 neck_01 {parent=spine_05, loc=<11.3138 0.1529 0>, rot=<0 0 -0.2003 -0.9797>}
 neck_02 {parent=neck_01, loc=<4.5627 0 0>, rot=<0 0 0.0504 -0.9987>}
 head {parent=neck_02, loc=<5.3914 0 0>, rot=<0 0 0.115 -0.9934>}
 helper_head {parent=neck_02, loc=<0.6111 -20.3934 0>, rot=<0 0 0.115 -0.9934>}
 neck_02_helper_offset {parent=neck_01, loc=<4.5627 0 0>, rot=<0 0 0.0504 -0.9987>}
 neck_02_helper {parent=neck_02_helper_offset, loc=<0 0 0>, rot=<0 0 0 1>}
 base_jacketflapneck {parent=spine_05, loc=<11.3138 0.1529 0>, rot=<0 0 -0.2003 -0.9797>}
 base_jacket_flapneck_02_r {parent=base_jacketflapneck, loc=<-1.5551 1.1113 6.592>, rot=<0.0321 0.1968 0.1576 -0.9672>}
 dyn_jacket_flapneck_02_r {parent=base_jacket_flapneck_02_r, loc=<0 0 0>, rot=<0 0 0 1>}
 base_jacket_flapneck_02_l {parent=base_jacketflapneck, loc=<-1.5551 1.1113 -6.592>, rot=<-0.0321 -0.1968 0.1576 -0.9672>}
 dyn_jacket_flapneck_02_l {parent=base_jacket_flapneck_02_l, loc=<0 0 0>, rot=<0 0 0 1>}
 base_jacket_flapneck_01_l {parent=base_jacketflapneck, loc=<-2.1744 6.2576 -6.7958>, rot=<-0.051 -0.3132 0.1525 -0.936>}
 dyn_jacket_flapneck_01_l {parent=base_jacket_flapneck_01_l, loc=<0 0 0>, rot=<0 0 0 1>}
 base_jacket_flapneck_01_r {parent=base_jacketflapneck, loc=<-2.1744 6.2576 6.7958>, rot=<0.051 0.3132 0.1525 -0.936>}
 dyn_jacket_flapneck_01_r {parent=base_jacket_flapneck_01_r, loc=<0 0 0>, rot=<0 0 0 1>}
 base_jacket_flapcn {parent=base_jacketflapneck, loc=<-1.8065 -5.2051 0>, rot=<0 0 0.1608 -0.987>}
 dyn_jacket_flapcn {parent=base_jacket_flapcn, loc=<0 0 0>, rot=<0 0 0 1>}
 clavicle_l {parent=spine_05, loc=<5.4831 -0.1053 -3.2946>, rot=<0.0465 -0.7289 -0.009 -0.6829>}
 upperarm_l {parent=clavicle_l, loc=<9.8658 -1.2136 -0.4105>, rot=<0.0164 -0.3245 -0.0269 -0.9454>}
 lowerarm_l {parent=upperarm_l, loc=<26.465 -0.8345 0.259>, rot=<0.0036 0.0161 -0.2298 -0.9731>}
 hand_l {parent=lowerarm_l, loc=<22.7355 -0.3203 -0.6105>, rot=<-0.6784 -0.0223 0.0735 -0.7307>}
 hand_attach_l {parent=hand_l, loc=<8.5021 -3.0001 -0.0001>, rot=<0.7071 0.7071 -0.0001 0.0002>}
 weapon_l {parent=hand_l, loc=<-0.1965 -3.4115 0.3812>, rot=<0.0001 -0.0001 -0.001 -1>}
 thumb_01_l {parent=hand_l, loc=<1.3118 -0.3191 -2.459>, rot=<0.6546 -0.4042 0.0664 -0.6354>}
 thumb_02_l {parent=thumb_01_l, loc=<5.9336 -0.0448 0.0401>, rot=<-0.0037 0.0213 0.298 -0.9543>}
 thumb_03_l {parent=thumb_02_l, loc=<2.2223 0.3682 -0.0907>, rot=<0.0118 -0.013 0.0066 -0.9998>}
 index_metacarpal_l {parent=hand_l, loc=<1.6535 0.0183 -2.96>, rot=<0.0158 -0.0124 -0.0173 -0.9996>}
 index_01_l {parent=index_metacarpal_l, loc=<8.3008 0.0357 -0.0524>, rot=<-0.0044 0.0485 0.0924 -0.9945>}
 index_02_l {parent=index_01_l, loc=<3.3587 -0.3004 0.0452>, rot=<-0.002 0.0027 0.3225 -0.9466>}
 index_03_l {parent=index_02_l, loc=<2.5578 -0.123 -0.0252>, rot=<-0.002 0.0067 0.2558 -0.9667>}
 middle_metacarpal_l {parent=hand_l, loc=<2.1856 -0.1725 -0.5984>, rot=<0.006 0.0168 -0.0329 -0.9993>}
 middle_01_l {parent=middle_metacarpal_l, loc=<7.3275 -0.2447 0.0129>, rot=<-0.0208 0.0042 0.0858 -0.9961>}
 middle_02_l {parent=middle_01_l, loc=<4.0123 0.1401 0.1175>, rot=<-0.0218 0.0258 0.3512 -0.9357>}
 middle_03_l {parent=middle_02_l, loc=<2.6077 0.0534 0.1373>, rot=<0.0082 0.0188 0.1708 -0.9851>}
 ring_metacarpal_l {parent=hand_l, loc=<2.2074 -0.3625 1.3638>, rot=<-0.0284 0.0722 -0.0458 -0.9959>}
 ring_01_l {parent=ring_metacarpal_l, loc=<7.0295 -0.6197 -0.1446>, rot=<-0.0489 -0.0252 0.1892 -0.9804>}
 ring_02_l {parent=ring_01_l, loc=<3.6174 0.4647 -0.1027>, rot=<0 0 0.2806 -0.9598>}
 ring_03_l {parent=ring_02_l, loc=<2.4152 0.114 0.0199>, rot=<0 0 0.1189 -0.9929>}
 pinky_metacarpal_l {parent=hand_l, loc=<1.7616 -0.9753 3.3814>, rot=<-0.0397 0.0814 -0.069 -0.9935>}
 pinky_01_l {parent=pinky_metacarpal_l, loc=<7.1709 -0.4831 -0.1094>, rot=<-0.0654 -0.0035 0.2242 -0.9723>}
 pinky_02_l {parent=pinky_01_l, loc=<2.8274 0.0893 -0.1043>, rot=<0.0037 0.0011 0.2854 -0.9584>}
 pinky_03_l {parent=pinky_02_l, loc=<2.2046 0.0307 -0.0078>, rot=<-0.0002 0.0098 0.1445 -0.9895>}
 deform_wrist_dn_base_l {parent=hand_l, loc=<0 -1.5326 0>, rot=<0 0 0 -1>}
 deform_wrist_dn_l {parent=deform_wrist_dn_base_l, loc=<0 0 0>, rot=<0 0 0 1>}
 deform_wrist_up_base_l {parent=hand_l, loc=<0 2.3009 0>, rot=<0 0 0 -1>}
 deform_wrist_up_l {parent=deform_wrist_up_base_l, loc=<0 0 0>, rot=<0 0 0 1>}
 lowerarm_twist_02_l {parent=lowerarm_l, loc=<10.6707 0 0>, rot=<0 0 0 1>}
 lowerarm_twist_01_l {parent=lowerarm_l, loc=<18.316 0 0>, rot=<0 0 0 1>}
 deform_elbow_in_base_l {parent=lowerarm_l, loc=<0.1834 3.6062 -0.2572>, rot=<0 0 0 -1>}
 deform_elbow_in_l {parent=deform_elbow_in_base_l, loc=<0 0 0>, rot=<0 0 0 1>}
 deform_elbow_out_base_l {parent=lowerarm_l, loc=<0 -4 0>, rot=<0 0 0 1>}
 deform_elbow_out_l {parent=deform_elbow_out_base_l, loc=<0 0 0>, rot=<0 0 0 1>}
 upperarm_twist_01_l {parent=upperarm_l, loc=<7.2135 -0.1134 0>, rot=<0 0 0 1>}
 deform_upperarm_bk_base_l {parent=upperarm_twist_01_l, loc=<-6.4053 -4.3746 0.314>, rot=<0 0 0 -1>}
 deform_upperarm_bk_l {parent=deform_upperarm_bk_base_l, loc=<0 0 0>, rot=<0 0 0 1>}
 deform_upperarm_fr_base_l {parent=upperarm_twist_01_l, loc=<-5.6312 5.092 0.3354>, rot=<0 0 0 -1>}
 deform_upperarm_fr_l {parent=deform_upperarm_fr_base_l, loc=<0 0 0>, rot=<0 0 0 1>}
 deform_upperarm_up_base_l {parent=upperarm_twist_01_l, loc=<-8.891 0.4825 3.9517>, rot=<0 0 0 -1>}
 deform_upperarm_up_l {parent=deform_upperarm_up_base_l, loc=<0 0 0>, rot=<0 0 0 1>}
 upperarm_twist_02_l {parent=upperarm_l, loc=<19.7998 -0.1134 0>, rot=<0 0 0 1>}
 clavicle_r {parent=spine_05, loc=<5.4831 -0.1053 3.2946>, rot=<0.7289 0.0465 -0.6829 0.009>}
 upperarm_r {parent=clavicle_r, loc=<-9.8658 1.2136 0.4105>, rot=<0.0164 -0.3245 -0.0269 -0.9454>}
 lowerarm_r {parent=upperarm_r, loc=<-26.465 0.8345 -0.259>, rot=<0.0036 0.0161 -0.2298 -0.9731>}
 hand_r {parent=lowerarm_r, loc=<-22.7355 0.3203 0.6106>, rot=<-0.6784 -0.0223 0.0735 -0.7307>}
 thumb_01_r {parent=hand_r, loc=<-1.3118 0.3191 2.459>, rot=<0.6546 -0.4042 0.0664 -0.6354>}
 thumb_02_r {parent=thumb_01_r, loc=<-5.9336 0.0448 -0.0401>, rot=<-0.0037 0.0213 0.298 -0.9543>}
 thumb_03_r {parent=thumb_02_r, loc=<-2.2223 -0.3682 0.0907>, rot=<0.0118 -0.013 0.0066 -0.9998>}
 index_metacarpal_r {parent=hand_r, loc=<-1.6535 -0.0183 2.96>, rot=<0.0158 -0.0124 -0.0173 -0.9996>}
 index_01_r {parent=index_metacarpal_r, loc=<-8.3008 -0.0357 0.0524>, rot=<-0.0044 0.0485 0.0924 -0.9945>}
 index_02_r {parent=index_01_r, loc=<-3.3587 0.3004 -0.0452>, rot=<-0.002 0.0027 0.3225 -0.9466>}
 index_03_r {parent=index_02_r, loc=<-2.5578 0.123 0.0252>, rot=<-0.002 0.0067 0.2558 -0.9667>}
 middle_metacarpal_r {parent=hand_r, loc=<-2.1856 0.1725 0.5984>, rot=<0.006 0.0168 -0.0329 -0.9993>}
 middle_01_r {parent=middle_metacarpal_r, loc=<-7.3275 0.2447 -0.0129>, rot=<-0.0208 0.0042 0.0858 -0.9961>}
 middle_02_r {parent=middle_01_r, loc=<-4.0123 -0.1401 -0.1175>, rot=<-0.0218 0.0258 0.3512 -0.9357>}
 middle_03_r {parent=middle_02_r, loc=<-2.6077 -0.0534 -0.1373>, rot=<0.0082 0.0188 0.1708 -0.9851>}
 ring_metacarpal_r {parent=hand_r, loc=<-2.2074 0.3625 -1.3638>, rot=<-0.0284 0.0722 -0.0458 -0.9959>}
 ring_01_r {parent=ring_metacarpal_r, loc=<-7.0295 0.6197 0.1446>, rot=<-0.0489 -0.0252 0.1892 -0.9804>}
 ring_02_r {parent=ring_01_r, loc=<-3.6174 -0.4647 0.1027>, rot=<0 0 0.2806 -0.9598>}
 ring_03_r {parent=ring_02_r, loc=<-2.4152 -0.114 -0.0199>, rot=<0 0 0.1189 -0.9929>}
 pinky_metacarpal_r {parent=hand_r, loc=<-1.7616 0.9753 -3.3814>, rot=<-0.0397 0.0814 -0.069 -0.9935>}
 pinky_01_r {parent=pinky_metacarpal_r, loc=<-7.1709 0.4831 0.1094>, rot=<-0.0654 -0.0035 0.2242 -0.9723>}
 pinky_02_r {parent=pinky_01_r, loc=<-2.8274 -0.0893 0.1043>, rot=<0.0037 0.0011 0.2854 -0.9584>}
 pinky_03_r {parent=pinky_02_r, loc=<-2.2046 -0.0307 0.0078>, rot=<-0.0002 0.0098 0.1445 -0.9895>}
 weapon_r {parent=hand_r, loc=<0.1974 3.4111 -0.3811>, rot=<0.4128 -0.4424 0.5877 -0.5372>}
 hand_attach_r {parent=hand_r, loc=<-10.9846 2.9733 -0.8686>, rot=<0.0066 -0.1117 -0.7318 -0.6723>}
 deform_wrist_dn_base_r {parent=hand_r, loc=<-0.0001 1.5328 0>, rot=<0 0 0 -1>}
 deform_wrist_dn_r {parent=deform_wrist_dn_base_r, loc=<0 0 0>, rot=<0 0 0 1>}
 deform_wrist_up_base_r {parent=hand_r, loc=<-0.0002 -2.3007 -0.0001>, rot=<0 0 0 -1>}
 deform_wrist_up_r {parent=deform_wrist_up_base_r, loc=<0 0 0>, rot=<0 0 0 1>}
 lowerarm_twist_02_r {parent=lowerarm_r, loc=<-10.6707 0 0>, rot=<0 0 0 -1>}
 lowerarm_twist_01_r {parent=lowerarm_r, loc=<-18.316 0 0>, rot=<0 0 0 -1>}
 deform_elbow_in_base_r {parent=lowerarm_r, loc=<-0.1834 -3.6062 0.2572>, rot=<0 0 0 -1>}
 deform_elbow_in_r {parent=deform_elbow_in_base_r, loc=<0 0 0>, rot=<0 0 0 1>}
 deform_elbow_out_base_r {parent=lowerarm_r, loc=<0 4 -0.0001>, rot=<0 0 0 -1>}
 deform_elbow_out_r {parent=deform_elbow_out_base_r, loc=<0 0 0>, rot=<0 0 0 1>}
 upperarm_twist_01_r {parent=upperarm_r, loc=<-7.2135 0.1134 0>, rot=<0 0 0 -1>}
 deform_upperarm_bk_base_r {parent=upperarm_twist_01_r, loc=<6.4053 4.3746 -0.314>, rot=<0 0 0 -1>}
 deform_upperarm_bk_r {parent=deform_upperarm_bk_base_r, loc=<0 0 0>, rot=<0 0 0 1>}
 deform_upperarm_fr_base_r {parent=upperarm_twist_01_r, loc=<5.6312 -5.092 -0.3354>, rot=<0 0 0 -1>}
 deform_upperarm_fr_r {parent=deform_upperarm_fr_base_r, loc=<0 0 0>, rot=<0 0 0 1>}
 deform_upperarm_up_base_r {parent=upperarm_twist_01_r, loc=<8.891 -0.4825 -3.9517>, rot=<0 0 0 -1>}
 deform_upperarm_up_r {parent=deform_upperarm_up_base_r, loc=<0 0 0>, rot=<0 0 0 1>}
 helper_upperarm_r {parent=deform_upperarm_up_r, loc=<-1.3449 0.1988 3.63>, rot=<0 0 0 1>}
 upperarm_twist_02_r {parent=upperarm_r, loc=<-19.7998 0.1134 0>, rot=<0 0 0 -1>}
 attach_backpack {parent=spine_05, loc=<-0.12 -13.4742 0.0001>, rot=<-0.0093 -0.707 0.0092 -0.707>}
 Backpack_BR {parent=attach_backpack, type=Null, loc=<0 0 0>, rot=<0 0 0 1>}
 attach_cape {parent=spine_05, loc=<9.0208 -11.715 0.0002>, rot=<0.0616 -0.7044 -0.0616 -0.7044>}
 cape {parent=attach_cape, type=Null, loc=<0 0 0>, rot=<0 0 0 1>}
 head_null {parent=spine_05, type=Null, loc=<20.6573 3.5489 0>, rot=<0 0 -0.036 -0.9994>}
 clavicle_l_null {parent=spine_05, type=Null, loc=<26.3226 -20.5799 0.6412>, rot=<-0.0859 0.3428 0.1195 0.9278>}
 clavicle_r_null {parent=spine_05, type=Null, loc=<5.483 -0.1053 3.2946>, rot=<0.7289 0.0465 -0.6829 0.009>}
 upperarm_r_null {parent=spine_05, type=Null, loc=<4.5519 -1.8922 13.0371>, rot=<-0.9118 -0.0385 0.4081 -0.0238>}
 deform_pec_aim_base_l {parent=spine_04, loc=<3.2456 9.3309 -1.2237>, rot=<0.0603 -0.4033 0.3318 -0.8507>}
 deform_pec_aim_l {parent=deform_pec_aim_base_l, loc=<0 0 0>, rot=<0 0 0 1>}
 deform_pec_base_l {parent=deform_pec_aim_l, loc=<9.3665 0 0>, rot=<0 0 0 1>}
 deform_pec_l {parent=deform_pec_base_l, loc=<0 0 0>, rot=<0 0 0 1>}
 deform_pec_aim_base_r {parent=spine_04, loc=<3.2452 9.3309 1.2237>, rot=<-0.4033 -0.0603 0.8507 0.3318>}
 deform_pec_aim_r {parent=deform_pec_aim_base_r, loc=<0 0 0>, rot=<0 0 0 1>}
 deform_pec_base_r {parent=deform_pec_aim_r, loc=<-9.3667 -0.0003 -0.0004>, rot=<0 0 0 1>}
 deform_pec_r {parent=deform_pec_base_r, loc=<0 0 0>, rot=<0 0 0 1>}
 base_jacket {parent=spine_04, loc=<0 0 0>, rot=<0 0 0 1>}
 base_jacket_r_2 {parent=base_jacket, loc=<1.1069 -11.3698 8.9099>, rot=<0.0208 -0.4996 0.8653 -0.036>}
 dyn_jacket_r_2 {parent=base_jacket_r_2, loc=<0 0 0>, rot=<0 0 0 1>}
 base_jacket_r_1 {parent=base_jacket, loc=<-3.1887 -2.0232 14.2734>, rot=<0 -0.0004 0.9991 -0.0416>}
 dyn_jacket_r_1 {parent=base_jacket_r_1, loc=<0 0 0>, rot=<0 0 0 1>}
 base_jacket_r_0 {parent=base_jacket, loc=<-2.9912 7.0157 10.9016>, rot=<-0.0188 0.4518 0.8911 -0.0371>}
 dyn_jacket_r_0 {parent=base_jacket_r_0, loc=<0 0 0>, rot=<0 0 0 1>}
 base_jacket_back {parent=base_jacket, loc=<2.9693 -12.753 0>, rot=<0.0294 -0.7065 0.7065 -0.0294>}
 dyn_jacket_back {parent=base_jacket_back, loc=<0 0 0>, rot=<0 0 0 1>}
 base_jacket_l_2 {parent=base_jacket, loc=<1.1067 -11.3698 -8.9099>, rot=<0.036 -0.8653 0.4996 -0.0208>}
 dyn_jacket_l_2 {parent=base_jacket_l_2, loc=<0 0 0>, rot=<0 0 0 1>}
 base_jacket_l_1 {parent=base_jacket, loc=<-3.1888 -2.0248 -14.0948>, rot=<0.0416 -0.9991 0.0004 0>}
 dyn_jacket_l_1 {parent=base_jacket_l_1, loc=<0 0 0>, rot=<0 0 0 1>}
 base_jacket_l_0 {parent=base_jacket, loc=<-3.0149 7.2976 -11.0882>, rot=<0.0371 -0.8911 -0.4518 0.0188>}
 dyn_jacket_l_0 {parent=base_jacket_l_0, loc=<0 0 0>, rot=<0 0 0 1>}
 base_zipper_l {parent=dyn_jacket_l_0, loc=<4.9529 3.9294 2.2878>, rot=<0.6383 0.038 0.0143 -0.7687>}
 dyn_zipper_l {parent=base_zipper_l, loc=<0 0 0>, rot=<0 0 0 1>}
 base_bottle_l {parent=spine_01, loc=<4.7831 -5.3687 -10.2296>, rot=<0.1608 -0.9366 0.0172 -0.3108>}
 dyn_bottle_l {parent=base_bottle_l, loc=<0 0 0>, rot=<0 0 0 1>}
 base_burbblebottle_08 {parent=dyn_bottle_l, loc=<11.5977 -0.6051 -1.1463>, rot=<-0.0873 0.4445 -0.1187 -0.8836>}
 dyn_bubblebottle_08 {parent=base_burbblebottle_08, loc=<0 0 0>, rot=<0 0 0 1>}
 base_burbblebottle_07 {parent=dyn_bottle_l, loc=<11.8211 0.2791 -2.9518>, rot=<-0.0873 0.4445 -0.1187 -0.8836>}
 dyn_bubblebottle_07 {parent=base_burbblebottle_07, loc=<0 0 0>, rot=<0 0 0 1>}
 base_burbblebottle_06 {parent=dyn_bottle_l, loc=<10.9563 1.0023 -0.2091>, rot=<-0.0873 0.4445 -0.1187 -0.8836>}
 dyn_bubblebottle_06 {parent=base_burbblebottle_06, loc=<0 0 0>, rot=<0 0 0 1>}
 base_burbblebottle_05 {parent=dyn_bottle_l, loc=<10.392 1.5274 -3.836>, rot=<-0.0873 0.4445 -0.1187 -0.8836>}
 dyn_bubblebottle_05 {parent=base_burbblebottle_05, loc=<0 0 0>, rot=<0 0 0 1>}
 base_burbblebottle_04 {parent=dyn_bottle_l, loc=<7.9382 0.2464 -1.0696>, rot=<-0.0873 0.4445 -0.1187 -0.8836>}
 dyn_bubblebottle_04 {parent=base_burbblebottle_04, loc=<0 0 0>, rot=<0 0 0 1>}
 base_burbblebottle_03 {parent=dyn_bottle_l, loc=<5.1245 1.6717 0.2697>, rot=<-0.0873 0.4445 -0.1187 -0.8836>}
 dyn_bubblebottle_03 {parent=base_burbblebottle_03, loc=<0 0 0>, rot=<0 0 0 1>}
 base_burbblebottle_02 {parent=dyn_bottle_l, loc=<4.1184 -1.1277 -1.0936>, rot=<-0.0873 0.4445 -0.1187 -0.8836>}
 dyn_bubblebottle_02 {parent=base_burbblebottle_02, loc=<0 0 0>, rot=<0 0 0 1>}
 base_burbblebottle_01 {parent=dyn_bottle_l, loc=<3.7707 1.31 0.8209>, rot=<-0.0873 0.4445 -0.1187 -0.8836>}
 dyn_bubblebottle_01 {parent=base_burbblebottle_01, loc=<0 0 0>, rot=<0 0 0 1>}
 thigh_l {parent=pelvis, loc=<-1.0855 -0.2109 -8.7053>, rot=<0.0009 0.0355 0.0155 -0.9993>}
 calf_l {parent=thigh_l, loc=<-42.5284 -0.0019 -0.7297>, rot=<0.0006 -0.0161 -0.0391 -0.9991>}
 foot_l {parent=calf_l, loc=<-46.7401 -2.5014 -0.592>, rot=<0.0427 -0.0238 0.0353 -0.9982>}
 ball_l {parent=foot_l, loc=<-3.5658 10.8066 -0.1544>, rot=<0.0001 -0.0001 -0.7063 -0.7079>}
 base_shoes_02_l {parent=foot_l, loc=<1.2812 7.3055 -3.7051>, rot=<-0.0359 -0.9812 -0.069 -0.1765>}
 dyn_shoes_a_02_l {parent=base_shoes_02_l, loc=<0 0 0>, rot=<0 0 0 1>}
 dyn_shoes_b_02_l {parent=dyn_shoes_a_02_l, loc=<3.6712 0 0.1476>, rot=<0 0 0 1>}
 dyn_shoes_c_02_l {parent=dyn_shoes_b_02_l, loc=<3.1924 0.1552 -0.3163>, rot=<0 0 0 1>}
 base_shoes_01_l {parent=foot_l, loc=<2.4131 6.7483 3.2934>, rot=<-0.0792 -0.7052 -0.6999 0.081>}
 dyn_shoes_a_01_l {parent=base_shoes_01_l, loc=<0 0 0>, rot=<0 0 0 1>}
 dyn_shoes_b_01_l {parent=dyn_shoes_a_01_l, loc=<4.034 0.9634 -0.2526>, rot=<0 0 0 1>}
 dyn_shoes_c_01_l {parent=dyn_shoes_b_01_l, loc=<3.5266 -0.0122 0.0073>, rot=<0 0 0 1>}
 base_shoes_l {parent=foot_l, loc=<3.7593 6.8723 0.3086>, rot=<-0.503 -0.4946 -0.5001 -0.5023>}
 base_shoes_cn_01_l {parent=base_shoes_l, loc=<0 0 0>, rot=<0 0 0 1>}
 dyn_shoes_cn_01_l {parent=base_shoes_cn_01_l, loc=<0 0 0>, rot=<0 0 0 1>}
 base_shoes_cn_02_l {parent=base_shoes_l, loc=<0 0 0>, rot=<0 0 0 1>}
 dyn_shoes_cn_02_l {parent=base_shoes_cn_02_l, loc=<0 0 0>, rot=<0 0 0 1>}
 calf_twist_02_l {parent=calf_l, loc=<-19.1546 -1.7943 -0.2638>, rot=<0.0237 0.0095 -0.0035 -0.9997>}
 calf_twist_01_l {parent=calf_l, loc=<-35.8483 -1.9197 -0.5794>, rot=<0.0237 0.0095 -0.0035 -0.9997>}
 deform_knee_in_base_l {parent=calf_l, loc=<0.8031 -5.0231 0.2456>, rot=<0 0 0 -1>}
 deform_knee_in_l {parent=deform_knee_in_base_l, loc=<0 0 0>, rot=<0 0 0 1>}
 deform_knee_out_base_l {parent=calf_l, loc=<0 4 0>, rot=<0 0 0 -1>}
 deform_knee_out_l {parent=deform_knee_out_base_l, loc=<0 0 0>, rot=<0 0 0 1>}
 thigh_twist_01_l {parent=thigh_l, loc=<-27.952 0 0>, rot=<0.0013 0 0 -1>}
 deform_groin_base_l {parent=thigh_l, loc=<0.4801 5.6776 -0.2486>, rot=<0 0 0 -1>}
 deform_groin_l {parent=deform_groin_base_l, loc=<0 0 0>, rot=<0 0 0 1>}
 thigh_r {parent=pelvis, loc=<-1.0855 -0.2109 8.7053>, rot=<0.0355 -0.0009 0.9993 0.0155>}
 calf_r {parent=thigh_r, loc=<42.5284 0.0019 0.7297>, rot=<0.0006 -0.0161 -0.0391 -0.9991>}
 foot_r {parent=calf_r, loc=<46.7401 2.5014 0.592>, rot=<0.0427 -0.0238 0.0353 -0.9982>}
 ball_r {parent=foot_r, loc=<3.5658 -10.8066 0.1544>, rot=<0.0001 -0.0001 -0.7063 -0.7079>}
 base_shoes_01_r {parent=foot_r, loc=<-1.3969 -7.6489 -5.1663>, rot=<0.9357 -0.0992 -0.105 -0.3218>}
 dyn_shoes_a_01_r {parent=base_shoes_01_r, loc=<0 0 0>, rot=<0 0 0 1>}
 dyn_shoes_b_01_r {parent=dyn_shoes_a_01_r, loc=<3.8104 0.1721 -0.0593>, rot=<0 0 0 1>}
 dyn_shoes_c_01_r {parent=dyn_shoes_b_01_r, loc=<3.2301 0 -0.0606>, rot=<0 0 0 1>}
 base_shoes_02_r {parent=foot_r, loc=<-1.1564 -8.316 2.4209>, rot=<0.5852 0.1188 0.1747 -0.7829>}
 dyn_shoes_a_02_r {parent=base_shoes_02_r, loc=<0 0 0>, rot=<0 0 0 1>}
 dyn_shoes_b_02_r {parent=dyn_shoes_a_02_r, loc=<3.6282 -0.2542 0.0837>, rot=<0 0 0 1>}
 dyn_shoes_c_02_r {parent=dyn_shoes_b_02_r, loc=<3.3696 0.4627 -0.2416>, rot=<0 0 0 1>}
 base_shoes_r {parent=foot_r, loc=<-3.9331 -7.9226 -1.6189>, rot=<0.4617 -0.553 -0.547 0.4263>}
 base_shoes_cn_01_r {parent=base_shoes_r, loc=<0 0 0>, rot=<0 0 0 1>}
 dyn_shoes_cn_01_r {parent=base_shoes_cn_01_r, loc=<0 0 0>, rot=<0 0 0 1>}
 base_shoes_cn_02_r {parent=base_shoes_r, loc=<0 0 0>, rot=<0 0 0 1>}
 dyn_shoes_cn_02_r {parent=base_shoes_cn_02_r, loc=<0 0 0>, rot=<0 0 0 1>}
 calf_twist_02_r {parent=calf_r, loc=<19.1546 1.7943 0.2638>, rot=<0.0237 0.0095 -0.0035 -0.9997>}
 calf_twist_01_r {parent=calf_r, loc=<35.8483 1.9197 0.5794>, rot=<0.0237 0.0095 -0.0035 -0.9997>}
 deform_knee_in_base_r {parent=calf_r, loc=<-0.8031 5.0231 -0.2457>, rot=<0 0 0 -1>}
 deform_knee_in_r {parent=deform_knee_in_base_r, loc=<0 0 0>, rot=<0 0 0 1>}
 deform_knee_out_base_r {parent=calf_r, loc=<0 -4 0>, rot=<0 0 0 -1>}
 deform_knee_out_r {parent=deform_knee_out_base_r, loc=<0 0 0>, rot=<0 0 0 1>}
 thigh_twist_01_r {parent=thigh_r, loc=<27.952 0 0>, rot=<0.0013 0 0 -1>}
 deform_groin_base_r {parent=thigh_r, loc=<-0.4801 -5.6776 0.2486>, rot=<0 0 0 -1>}
 deform_groin_r {parent=deform_groin_base_r, loc=<0 0 0>, rot=<0 0 0 1>}
 deform_glute_base_l {parent=pelvis, loc=<-7.3466 -9.0213 -9.1245>, rot=<0.0172 0.031 -0.4477 -0.8935>}
 deform_glute_l {parent=deform_glute_base_l, loc=<0 0 0>, rot=<0 0 0 1>}
 deform_glute_base_r {parent=pelvis, loc=<-7.3464 -9.0213 9.1245>, rot=<0.031 -0.0172 0.8935 -0.4477>}
 deform_glute_r {parent=deform_glute_base_r, loc=<0 0 0>, rot=<0 0 0 1>}
 attach {parent=root, loc=<0 0 0>, rot=<0 0 0.8243 -0.5662>}
 ik_hand_root {parent=root, loc=<0 0 0>, rot=<0 0 0.7569 -0.6535>}
 ik_hand_gun {parent=ik_hand_root, loc=<3.7099 -48.1482 104.4381>, rot=<0.6276 0.1309 0.4183 -0.6434>}
 ik_hand_r {parent=ik_hand_gun, loc=<0 0 0>, rot=<0 0 0 1>}
 ik_hand_l {parent=ik_hand_gun, loc=<-44.2996 -23.0017 82.4105>, rot=<-0.6158 -0.7027 -0.3565 0>}
 ik_foot_root {parent=root, loc=<-0.0537 0.0049 0>, rot=<0.0005 -0.0011 0.8243 -0.5662>}
 ik_foot_l {parent=ik_foot_root, loc=<1.5258 16.5093 8.3374>, rot=<-0.635 0.2086 0.6875 0.2839>}
 ik_foot_r {parent=ik_foot_root, loc=<12.1679 -11.1675 8.2729>, rot=<0.5159 -0.017 0.5619 -0.6464>}
 root_fp {parent=root, loc=<0 0 0>, rot=<0 0 0.8684 -0.4959>}
 attach_fp {parent=root_fp, loc=<-0.051 -0.0822 -0.505>, rot=<0.0068 -0.0123 0.0002 -0.9999>}
 camera_root {parent=root_fp, loc=<0 0 0>, rot=<0 0 0 1>}
 camera {parent=camera_root, loc=<0 0 0>, rot=<0 0 0 1>}
 attach_emote_01 {parent=root, loc=<0 0 0>, rot=<0 0 0.8684 -0.4959>}
 attach_emote_02 {parent=root, loc=<0 0 0>, rot=<0 0 0.8684 -0.4959>}
 attach_emote_03 {parent=root, loc=<0 0 0>, rot=<0 0 0.8684 -0.4959>}
 warp_point_01 {parent=root, loc=<0 0 0>, rot=<0 0 0.9177 -0.3972>}
 warp_point_02 {parent=root, loc=<0 0 0>, rot=<0 0 0.9177 -0.3972>}
 warp_point_03 {parent=root, loc=<0 0 0>, rot=<0 0 0.9177 -0.3972>}
 neck_01_ctrl {controlType=EULER_TRANSFORM, type=Control, loc=<0 0 0>, rot=<0 0 0 1>}
 neck_02_ctrl {controlType=EULER_TRANSFORM, parent=neck_01_ctrl, type=Control, loc=<0 0 0>, rot=<0 0 0 1>}
 head_ctrl {controlType=EULER_TRANSFORM, parent=neck_02_ctrl, type=Control, loc=<0 0 0>, rot=<0 0 0 1>}
 clavicle_l_ctrl {controlType=EULER_TRANSFORM, type=Control, loc=<0 0 0>, rot=<0 0 0 1>}
 clavicle_r_ctrl {controlType=EULER_TRANSFORM, type=Control, loc=<0 0 0>, rot=<0 0 0 1>}
]
variables [
 CrouchWalk {type=double, default=0}
 isHeadDownDelta {type=double, default=0}
 isHeadUpDelta {type=double, default=0}
]
