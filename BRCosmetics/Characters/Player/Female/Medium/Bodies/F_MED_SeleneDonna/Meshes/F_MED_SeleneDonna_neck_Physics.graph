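# Physics asset for the neck mesh of the female medium SeleneDonna body cosmetic, set up for its jacket collar flaps.
# Five dyn_ flap capsules each constrain to a small base_ bone body with tight swing limits.
# The center base body has locked rotation to base_jacketflapneck.
# Head, helper head, neck and right upper arm add larger capsules, with both head bodies rotating freely against the neck; no left arm body exists.
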PhysicsAsset /BRCosmetics/Characters/Player/Female/Medium/Bodies/F_MED_SeleneDonna/Meshes/F_MED_SeleneDonna_neck_Physics
bodies {
 base_jacket_flapcn [{shapeType=Capsule, length=0.505, radius=0.505, shapeName=base_jacket_flapcn_capsule, rotation=<rot 0 -90 0>}]
 base_jacket_flapneck_01_l [{shapeType=Capsule, length=0.505, radius=0.505, shapeName=base_jacket_flapneck_01_l_capsule, rotation=<rot 0 -90 0>}]
 base_jacket_flapneck_01_r [{shapeType=Capsule, length=0.505, radius=0.505, shapeName=base_jacket_flapneck_01_r_capsule, rotation=<rot 0 -90 0>}]
 base_jacket_flapneck_02_l [{shapeType=Capsule, length=0.505, radius=0.505, shapeName=base_jacket_flapneck_02_l_capsule, rotation=<rot 0 -90 0>}]
 base_jacket_flapneck_02_r [{shapeType=Capsule, length=0.505, radius=0.505, shapeName=base_jacket_flapneck_02_r_capsule, rotation=<rot 0 -90 0>}]
 base_jacketflapneck [{shapeType=Capsule, length=0.505, radius=0.505, shapeName=base_jacketflapneck_capsule, rotation=<rot 0 -90 0>}]
 dyn_jacket_flapcn [{shapeType=Capsule, length=10.3091, radius=2.5107, shapeName=dyn_jacket_flapcn_capsule, center=<4.3263 -0.6537 0>}]
 dyn_jacket_flapneck_01_l [{
   shapeType=Capsule
   length=2.5
   radius=3
   shapeName=dyn_jacket_flapneck_01_l_capsule
   center=<5.313 -0.3259 0>
   rotation=<rot -1.295 18.4623 108.0754>
  }]
 dyn_jacket_flapneck_01_r [{
   shapeType=Capsule
   length=2.5
   radius=3
   shapeName=dyn_jacket_flapneck_01_r_capsule
   center=<5.313 0.3259 0>
   rotation=<rot -1.295 18.4623 71.9246>
  }]
 dyn_jacket_flapneck_02_l [{
   shapeType=Capsule
   length=5
   radius=3
   shapeName=dyn_jacket_flapneck_02_l_capsule
   center=<4.7915 -1.9 0.3>
   rotation=<rot 1.2526 12.2997 118.7728>
  }]
 dyn_jacket_flapneck_02_r [{
   shapeType=Capsule
   length=5
   radius=3
   shapeName=dyn_jacket_flapneck_02_r_capsule
   center=<4.9147 -2 -0.2644>
   rotation=<rot -0.4899 13.0248 67.4477>
  }]
 head [{shapeType=Capsule, length=4.2344, radius=11.7932, shapeName=head_capsule, center=<4.952 2.5573 0>}]
 helper_head [{shapeType=Capsule, length=3.505, radius=13.1, shapeName=head_capsule}]
 neck_02 [{shapeType=Capsule, length=7, radius=5.8, shapeName=neck_02_capsule, center=<0 2.4271 0>, rotation=<rot 90 0 0>}]
 upperarm_r [{shapeType=Capsule, length=7, radius=6, shapeName=upperarm_r_capsule, center=<-6.5544 2.3262 0>, rotation=<rot 90 0 0>}]
}
constraints [
 {bone1Name=dyn_jacket_flapneck_02_r, bone2Name=base_jacket_flapneck_02_r, swing1LimitDegrees=10, swing2LimitDegrees=10, twistLimitDegrees=0}
 {bone1Name=dyn_jacket_flapneck_02_l, bone2Name=base_jacket_flapneck_02_l, swing1LimitDegrees=10, swing2LimitDegrees=10, twistLimitDegrees=0}
 {bone1Name=dyn_jacket_flapneck_01_l, bone2Name=base_jacket_flapneck_01_l, swing1LimitDegrees=10, swing2LimitDegrees=10, twistLimitDegrees=0}
 {bone1Name=dyn_jacket_flapneck_01_r, bone2Name=base_jacket_flapneck_01_r, swing1LimitDegrees=10, swing2LimitDegrees=10, twistLimitDegrees=0}
 {bone1Name=dyn_jacket_flapcn, bone2Name=base_jacket_flapcn, swing1LimitDegrees=10, swing2LimitDegrees=10, twistLimitDegrees=0}
 {
  bone1Name=base_jacket_flapcn
  bone2Name=base_jacketflapneck
  swing1LimitDegrees=45
  swing1Motion=Locked
  swing2LimitDegrees=45
  swing2Motion=Locked
  twistLimitDegrees=45
  twistMotion=Locked
 }
 {
  bone1Name=head
  bone2Name=neck_02
  swing1LimitDegrees=45
  swing1Motion=Free
  swing2LimitDegrees=45
  swing2Motion=Free
  twistLimitDegrees=45
  twistMotion=Free
 }
 {
  bone1Name=helper_head
  bone2Name=neck_02
  swing1LimitDegrees=45
  swing1Motion=Free
  swing2LimitDegrees=45
  swing2Motion=Free
  twistLimitDegrees=45
  twistMotion=Free
 }
]
