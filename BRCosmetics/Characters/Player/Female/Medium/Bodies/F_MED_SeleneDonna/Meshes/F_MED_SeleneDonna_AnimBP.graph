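# Post-process-style AnimBP for the F_MED_SeleneDonna cosmetic body, adding secondary motion to bubble bottles, shoes, jacket, zipper and hip bottle.
# It copies the parent mesh pose through a Control Rig, then chains cached stages of weighted bone constraints, AnimDynamics and RigidBody simulation.
# Jacket-back physics is skipped when non-cape, non-long backbling is worn.
# The update event computes gravity/sim settings, sine-wave "star" offsets, backbling flags and frontend alpha.
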AnimBlueprint /BRCosmetics/Characters/Player/Female/Medium/Bodies/F_MED_SeleneDonna/Meshes/F_MED_SeleneDonna_AnimBP
target_skeleton=&/F_MED_SeleneDonna_Skeleton
graphs [
 AnimGraph {nodes [
   n1 = "Constraint - base_bottle_l (PPPPP)"!(
    ComponentPose=Local To Component(LocalPose=Use cached pose 'Jacket_cache'())
    Alpha=1
    ConstraintWeights_0=0.6162
    ConstraintWeights_1:float=0.1881
    ConstraintWeights_2:float=0.1873
    ConstraintWeights_3:float=0.0069
    ConstraintWeights_4:float=0.0015
   ) -> (Pose) {
    alphaBoolBlend {blendOption=Linear}
    boneToModify {boneName=base_bottle_l}
    constraintSetup [
     {targetBone {boneName=pelvis}}
     {targetBone {boneName=spine_01}}
     {targetBone {boneName=spine_02}}
     {targetBone {boneName=spine_04}}
     {targetBone {boneName=spine_03}}
    ]
    constraintWeights [1,1,1,1,1]
   }
   n2 = Anim Dynamics - Bone: dyn_bottle_l!(
    SimulationSpace:AnimPhysSimSpaceType Enum
    ComponentPose=@n1
    Alpha=Get Alpha Frontend() -> (Alpha Frontend:double)
    GravityOverride:Vector="0,0,-980"
    ExternalForce="0,0,0"
   ) -> (Pose) {
    angularDampingOverride=1
    angularSpringConstant=8000
    bAngularSpring=true
    bOverrideAngularDamping=true
    bUseGravityOverride=true
    gravityScale=0.35
    simulationSpace=RootRelative
    alphaBoolBlend {blendOption=Linear}
    boundBone {boneName=dyn_bottle_l}
    componentAppliedLinearAccClamp=<3000 3000 3000>
    componentLinearAccScale=<0.8 0.8 0.8>
    componentLinearVelScale=<0.6 0.6 0.6>
    gravityOverride {z=-980}
    physicsBodyDefinitions [{
      boundBone {boneName=dyn_bottle_l}
      boxExtents=<20 30 30>
      constraintSetup {angularLimitsMax=<0 5 0>, angularLimitsMin=<0 -70 -55>, angularTarget=<1 0 0>}
      localJointOffset=<4 0 0>
     }]
   }
   Output Pose(Result=Component To Local(ComponentPose=@n2)) {layerGroup=None}
   n3 = Control Rig(Source=Copy Pose From Mesh() {bUseAttachedParent=true}) {
    alphaBoolBlend {blendOption=Linear}
    controlRigAssetReference {blueprintRigClass=&/F_MED_SeleneDonna_CtrlRig.F_MED_SeleneDonna_CtrlRig_C}
   }
   n4 = "Transform (Modify) Bone - Bone: dyn_bubblebottle_01"!(ComponentPose=Local To Component(LocalPose=@n3), Alpha=1, Translation="0,0,0", Rotation="0,0,0") -> (Pose) {
    rotationSpace=BCS_BoneSpace
    translationMode=BMM_Additive
    translationSpace=BCS_BoneSpace
    alphaBoolBlend {blendOption=Linear}
    boneToModify {boneName=dyn_bubblebottle_01}
   }
   n5 = Anim Dynamics - Bone: dyn_bubblebottle_01!(ComponentPose=@n4, Alpha=1, GravityOverride:Vector="0,0,-980") -> (Pose) {
    angularDampingOverride=1
    angularSpringConstant=20000
    bAngularSpring=true
    bGravityOverrideInSimSpace=true
    bOverrideAngularDamping=true
    bUseGravityOverride=true
    gravityScale=0.35
    simulationSpace=RootRelative
    alphaBoolBlend {blendOption=Linear}
    boundBone {boneName=dyn_bubblebottle_01}
    componentAppliedLinearAccClamp=<3000 3000 3000>
    componentLinearAccScale=<0.5 0.5 0.5>
    componentLinearVelScale=<0.5 0.5 0.5>
    gravityOverride {z=-980}
    physicsBodyDefinitions [{
      boundBone {boneName=dyn_bubblebottle_01}
      boxExtents=<35 35 35>
      constraintSetup {angularLimitsMax=<2 2 2>, angularLimitsMin=<-2 -2 -2>, angularTarget=<1 0 0>}
      localJointOffset=<8 0 0>
     }]
   }
   n6 = "Transform (Modify) Bone - Bone: dyn_bubblebottle_02"!(ComponentPose=@n5, Alpha=1, Translation="0,0,0", Rotation="0,0,0") -> (Pose) {
    rotationSpace=BCS_BoneSpace
    translationMode=BMM_Additive
    translationSpace=BCS_BoneSpace
    alphaBoolBlend {blendOption=Linear}
    boneToModify {boneName=dyn_bubblebottle_02}
   }
   n7 = Anim Dynamics - Bone: dyn_bubblebottle_02!(ComponentPose=@n6, Alpha=1, GravityOverride:Vector="0,0,-980") -> (Pose) {
    angularDampingOverride=1
    angularSpringConstant=20000
    bAngularSpring=true
    bGravityOverrideInSimSpace=true
    bOverrideAngularDamping=true
    bUseGravityOverride=true
    gravityScale=0.35
    simulationSpace=RootRelative
    alphaBoolBlend {blendOption=Linear}
    boundBone {boneName=dyn_bubblebottle_02}
    componentAppliedLinearAccClamp=<3000 3000 3000>
    componentLinearAccScale=<0.5 0.5 0.5>
    componentLinearVelScale=<0.5 0.5 0.5>
    gravityOverride {z=-980}
    physicsBodyDefinitions [{
      boundBone {boneName=dyn_bubblebottle_02}
      boxExtents=<35 35 35>
      constraintSetup {angularLimitsMax=<2 2 2>, angularLimitsMin=<-2 -2 -2>, angularTarget=<1 0 0>}
      localJointOffset=<8 0 0>
     }]
   }
   n8 = "Transform (Modify) Bone - Bone: dyn_bubblebottle_03"!(ComponentPose=@n7, Alpha=1, Translation="0,0,0", Rotation="0,0,0") -> (Pose) {
    rotationSpace=BCS_BoneSpace
    translationMode=BMM_Additive
    translationSpace=BCS_BoneSpace
    alphaBoolBlend {blendOption=Linear}
    boneToModify {boneName=dyn_bubblebottle_03}
   }
   n9 = Anim Dynamics - Bone: dyn_bubblebottle_03!(ComponentPose=@n8, Alpha=1, GravityOverride:Vector="0,0,-980") -> (Pose) {
    angularDampingOverride=1
    angularSpringConstant=20000
    bAngularSpring=true
    bGravityOverrideInSimSpace=true
    bOverrideAngularDamping=true
    bUseGravityOverride=true
    gravityScale=0.35
    simulationSpace=RootRelative
    alphaBoolBlend {blendOption=Linear}
    boundBone {boneName=dyn_bubblebottle_03}
    componentAppliedLinearAccClamp=<3000 3000 3000>
    componentLinearAccScale=<0.5 0.5 0.5>
    componentLinearVelScale=<0.5 0.5 0.5>
    gravityOverride {z=-980}
    physicsBodyDefinitions [{
      boundBone {boneName=dyn_bubblebottle_03}
      boxExtents=<35 35 35>
      constraintSetup {angularLimitsMax=<2 2 2>, angularLimitsMin=<-2 -2 -2>, angularTarget=<1 0 0>}
      localJointOffset=<8 0 0>
     }]
   }
   n10 = "Transform (Modify) Bone - Bone: dyn_bubblebottle_04"!(ComponentPose=@n9, Alpha=1, Translation="0,0,0", Rotation="0,0,0") -> (Pose) {
    rotationSpace=BCS_BoneSpace
    translationMode=BMM_Additive
    translationSpace=BCS_BoneSpace
    alphaBoolBlend {blendOption=Linear}
    boneToModify {boneName=dyn_bubblebottle_04}
   }
   n11 = Anim Dynamics - Bone: dyn_bubblebottle_04!(ComponentPose=@n10, Alpha=1, GravityOverride:Vector="0,0,-980") -> (Pose) {
    angularDampingOverride=1
    angularSpringConstant=20000
    bAngularSpring=true
    bGravityOverrideInSimSpace=true
    bOverrideAngularDamping=true
    bUseGravityOverride=true
    gravityScale=0.35
    simulationSpace=RootRelative
    alphaBoolBlend {blendOption=Linear}
    boundBone {boneName=dyn_bubblebottle_04}
    componentAppliedLinearAccClamp=<3000 3000 3000>
    componentLinearAccScale=<0.5 0.5 0.5>
    componentLinearVelScale=<0.5 0.5 0.5>
    gravityOverride {z=-980}
    physicsBodyDefinitions [{
      boundBone {boneName=dyn_bubblebottle_04}
      boxExtents=<35 35 35>
      constraintSetup {angularLimitsMax=<2 2 2>, angularLimitsMin=<-2 -2 -2>, angularTarget=<1 0 0>}
      localJointOffset=<8 0 0>
     }]
   }
   n12 = "Transform (Modify) Bone - Bone: dyn_bubblebottle_05"!(ComponentPose=@n11, Alpha=1, Translation="0,0,0", Rotation="0,0,0") -> (Pose) {
    rotationSpace=BCS_BoneSpace
    translationMode=BMM_Additive
    translationSpace=BCS_BoneSpace
    alphaBoolBlend {blendOption=Linear}
    boneToModify {boneName=dyn_bubblebottle_05}
   }
   n13 = Anim Dynamics - Bone: dyn_bubblebottle_05!(ComponentPose=@n12, Alpha=1, GravityOverride:Vector="0,0,-980") -> (Pose) {
    angularDampingOverride=1
    angularSpringConstant=20000
    bAngularSpring=true
    bGravityOverrideInSimSpace=true
    bOverrideAngularDamping=true
    bUseGravityOverride=true
    gravityScale=0.35
    simulationSpace=RootRelative
    alphaBoolBlend {blendOption=Linear}
    boundBone {boneName=dyn_bubblebottle_05}
    componentAppliedLinearAccClamp=<3000 3000 3000>
    componentLinearAccScale=<0.5 0.5 0.5>
    componentLinearVelScale=<0.5 0.5 0.5>
    gravityOverride {z=-980}
    physicsBodyDefinitions [{
      boundBone {boneName=dyn_bubblebottle_05}
      boxExtents=<35 35 35>
      constraintSetup {angularLimitsMax=<2 2 2>, angularLimitsMin=<-2 -2 -2>, angularTarget=<1 0 0>}
      localJointOffset=<8 0 0>
     }]
   }
   n14 = "Transform (Modify) Bone - Bone: dyn_bubblebottle_06"!(ComponentPose=@n13, Alpha=1, Translation="0,0,0", Rotation="0,0,0") -> (Pose) {
    rotationSpace=BCS_BoneSpace
    translationMode=BMM_Additive
    translationSpace=BCS_BoneSpace
    alphaBoolBlend {blendOption=Linear}
    boneToModify {boneName=dyn_bubblebottle_06}
   }
   n15 = Anim Dynamics - Bone: dyn_bubblebottle_06!(ComponentPose=@n14, Alpha=1, GravityOverride:Vector="0,0,-980") -> (Pose) {
    angularDampingOverride=1
    angularSpringConstant=20000
    bAngularSpring=true
    bGravityOverrideInSimSpace=true
    bOverrideAngularDamping=true
    bUseGravityOverride=true
    gravityScale=0.35
    simulationSpace=RootRelative
    alphaBoolBlend {blendOption=Linear}
    boundBone {boneName=dyn_bubblebottle_06}
    componentAppliedLinearAccClamp=<3000 3000 3000>
    componentLinearAccScale=<0.5 0.5 0.5>
    componentLinearVelScale=<0.5 0.5 0.5>
    gravityOverride {z=-980}
    physicsBodyDefinitions [{
      boundBone {boneName=dyn_bubblebottle_06}
      boxExtents=<35 35 35>
      constraintSetup {angularLimitsMax=<2 2 2>, angularLimitsMin=<-2 -2 -2>, angularTarget=<1 0 0>}
      localJointOffset=<8 0 0>
     }]
   }
   n16 = "Transform (Modify) Bone - Bone: dyn_bubblebottle_07"!(ComponentPose=@n15, Alpha=1, Translation="0,0,0", Rotation="0,0,0") -> (Pose) {
    rotationSpace=BCS_BoneSpace
    translationMode=BMM_Additive
    translationSpace=BCS_BoneSpace
    alphaBoolBlend {blendOption=Linear}
    boneToModify {boneName=dyn_bubblebottle_07}
   }
   n17 = Anim Dynamics - Bone: dyn_bubblebottle_07!(ComponentPose=@n16, Alpha=1, GravityOverride:Vector="0,0,-980") -> (Pose) {
    angularDampingOverride=1
    angularSpringConstant=20000
    bAngularSpring=true
    bGravityOverrideInSimSpace=true
    bOverrideAngularDamping=true
    bUseGravityOverride=true
    gravityScale=0.35
    simulationSpace=RootRelative
    alphaBoolBlend {blendOption=Linear}
    boundBone {boneName=dyn_bubblebottle_07}
    componentAppliedLinearAccClamp=<3000 3000 3000>
    componentLinearAccScale=<0.5 0.5 0.5>
    componentLinearVelScale=<0.5 0.5 0.5>
    gravityOverride {z=-980}
    physicsBodyDefinitions [{
      boundBone {boneName=dyn_bubblebottle_07}
      boxExtents=<35 35 35>
      constraintSetup {angularLimitsMax=<2 2 2>, angularLimitsMin=<-2 -2 -2>, angularTarget=<1 0 0>}
      localJointOffset=<8 0 0>
     }]
   }
   n18 = "Transform (Modify) Bone - Bone: dyn_bubblebottle_08"!(ComponentPose=@n17, Alpha=1, Translation="0,0,0", Rotation="0,0,0") -> (Pose) {
    rotationSpace=BCS_BoneSpace
    translationMode=BMM_Additive
    translationSpace=BCS_BoneSpace
    alphaBoolBlend {blendOption=Linear}
    boneToModify {boneName=dyn_bubblebottle_08}
   }
   n19 = Anim Dynamics - Bone: dyn_bubblebottle_08!(ComponentPose=@n18, Alpha=1, GravityOverride:Vector="0,0,-980") -> (Pose) {
    angularDampingOverride=1
    angularSpringConstant=20000
    bAngularSpring=true
    bGravityOverrideInSimSpace=true
    bOverrideAngularDamping=true
    bUseGravityOverride=true
    gravityScale=0.35
    simulationSpace=RootRelative
    alphaBoolBlend {blendOption=Linear}
    boundBone {boneName=dyn_bubblebottle_08}
    componentAppliedLinearAccClamp=<3000 3000 3000>
    componentLinearAccScale=<0.5 0.5 0.5>
    componentLinearVelScale=<0.5 0.5 0.5>
    gravityOverride {z=-980}
    physicsBodyDefinitions [{
      boundBone {boneName=dyn_bubblebottle_08}
      boxExtents=<35 35 35>
      constraintSetup {angularLimitsMax=<2 2 2>, angularLimitsMin=<-2 -2 -2>, angularTarget=<1 0 0>}
      localJointOffset=<8 0 0>
     }]
   }
   Save cached pose 'bubble_cache'(Pose=Component To Local(ComponentPose=@n19)) {cacheName=bubble_cache}
   n20 = "Constraint - base_shoes_01_r (PPP)"!(
    ComponentPose=Local To Component(LocalPose=Use cached pose 'bubble_cache'())
    Alpha=1
    ConstraintWeights_0=0.9238
    ConstraintWeights_1:float=0.0602
    ConstraintWeights_2:float=0.016
   ) -> (Pose) {
    alphaBoolBlend {blendOption=Linear}
    boneToModify {boneName=base_shoes_01_r}
    constraintSetup [{targetBone {boneName=foot_r}}, {targetBone {boneName=calf_twist_01_r}}, {targetBone {boneName=ball_r}}]
    constraintWeights [1,1,1]
   }
   n21 = "Constraint - base_shoes_02_r (PPP)"!(ComponentPose=@n20, Alpha=1, ConstraintWeights_0=0.9683, ConstraintWeights_1:float=0.0263, ConstraintWeights_2:float=0.0055) -> (Pose) {
    alphaBoolBlend {blendOption=Linear}
    boneToModify {boneName=base_shoes_02_r}
    constraintSetup [{targetBone {boneName=foot_r}}, {targetBone {boneName=ball_r}}, {targetBone {boneName=calf_twist_01_r}}]
    constraintWeights [1,1,1]
   }
   n22 = Anim Dynamics - Bone: base_shoes_02_r(ComponentPose=@n21, Alpha=1, ExternalForce="0,0,0") {
    angularDampingOverride=1
    angularSpringConstant=20000
    bAngularSpring=true
    bOverrideAngularDamping=true
    bUseGravityOverride=true
    gravityScale=0.35
    simulationSpace=RootRelative
    alphaBoolBlend {blendOption=Linear}
    boundBone {boneName=base_shoes_02_r}
    componentLinearAccScale=<1 1 1>
    componentLinearVelScale=<1 1 1>
    gravityOverride {z=-980}
    physicsBodyDefinitions [{
      boundBone {boneName=base_shoes_02_r}
      boxExtents=<20 30 30>
      constraintSetup {angularLimitsMax=<40 0 0>, angularLimitsMin=<-40 -40 0>, angularTarget=<1 0 0>}
      localJointOffset=<4 0 0>
     }]
   }
   n23 = Anim Dynamics - Bone: base_shoes_01_r(ComponentPose=@n22, Alpha=1, ExternalForce="0,0,0") {
    angularDampingOverride=1
    angularSpringConstant=20000
    bAngularSpring=true
    bOverrideAngularDamping=true
    bUseGravityOverride=true
    gravityScale=0.35
    simulationSpace=RootRelative
    alphaBoolBlend {blendOption=Linear}
    boundBone {boneName=base_shoes_01_r}
    componentLinearAccScale=<1 1 1>
    componentLinearVelScale=<1 1 1>
    gravityOverride {z=-980}
    physicsBodyDefinitions [{
      boundBone {boneName=base_shoes_01_r}
      boxExtents=<20 30 30>
      constraintSetup {angularLimitsMax=<40 0 0>, angularLimitsMin=<-40 -40 0>, angularTarget=<1 0 0>}
      localJointOffset=<4 0 0>
     }]
   }
   n24 = "Constraint - base_shoes_cn_01_r (PPP)"!(ComponentPose=@n23, Alpha=1, ConstraintWeights_0=0.8737, ConstraintWeights_1:float=0.1248, ConstraintWeights_2:float=0.0015) -> (Pose) {
    alphaBoolBlend {blendOption=Linear}
    boneToModify {boneName=base_shoes_cn_01_r}
    constraintSetup [{targetBone {boneName=foot_r}}, {targetBone {boneName=calf_twist_01_r}}, {targetBone {boneName=ball_r}}]
    constraintWeights [1,1,1]
   }
   n25 = "Constraint - base_shoes_cn_02_r (PPP)"!(ComponentPose=@n24, Alpha=1, ConstraintWeights_0=0.8737, ConstraintWeights_1:float=0.1248, ConstraintWeights_2:float=0.0015) -> (Pose) {
    alphaBoolBlend {blendOption=Linear}
    boneToModify {boneName=base_shoes_cn_02_r}
    constraintSetup [{targetBone {boneName=foot_r}}, {targetBone {boneName=calf_twist_01_r}}, {targetBone {boneName=ball_r}}]
    constraintWeights [1,1,1]
   }
   n26 = Anim Dynamics - Bone: dyn_shoes_cn_01_r(ComponentPose=@n25, Alpha=1, ExternalForce="0,0,0") {
    angularDampingOverride=1
    angularSpringConstant=20000
    bAngularSpring=true
    bOverrideAngularDamping=true
    bUseGravityOverride=true
    gravityScale=0.35
    simulationSpace=RootRelative
    alphaBoolBlend {blendOption=Linear}
    boundBone {boneName=dyn_shoes_cn_01_r}
    componentLinearAccScale=<1 1 1>
    componentLinearVelScale=<1 1 1>
    gravityOverride {z=-980}
    physicsBodyDefinitions [{
      boundBone {boneName=dyn_shoes_cn_01_r}
      boxExtents=<20 30 30>
      constraintSetup {angularLimitsMax=<30 15 0>, angularLimitsMin=<-30 -15 0>, angularTarget=<1 0 0>}
      localJointOffset=<4 0 0>
     }]
   }
   n27 = Anim Dynamics - Bone: dyn_shoes_cn_02_r(ComponentPose=@n26, Alpha=1, ExternalForce="0,0,0") {
    angularDampingOverride=1
    angularSpringConstant=20000
    bAngularSpring=true
    bOverrideAngularDamping=true
    bUseGravityOverride=true
    gravityScale=0.35
    simulationSpace=RootRelative
    alphaBoolBlend {blendOption=Linear}
    boundBone {boneName=dyn_shoes_cn_02_r}
    componentLinearAccScale=<1 1 1>
    componentLinearVelScale=<1 1 1>
    gravityOverride {z=-980}
    physicsBodyDefinitions [{
      boundBone {boneName=dyn_shoes_cn_02_r}
      boxExtents=<20 30 30>
      constraintSetup {angularLimitsMax=<30 15 0>, angularLimitsMin=<-30 -15 0>, angularTarget=<1 0 0>}
      localJointOffset=<4 0 0>
     }]
   }
   n28 = "Constraint - base_shoes_02_l (PPP)"!(ComponentPose=@n27, Alpha=1, ConstraintWeights_0=0.9683, ConstraintWeights_1:float=0.0263, ConstraintWeights_2:float=0.0055) -> (Pose) {
    alphaBoolBlend {blendOption=Linear}
    boneToModify {boneName=base_shoes_02_l}
    constraintSetup [{targetBone {boneName=foot_l}}, {targetBone {boneName=ball_l}}, {targetBone {boneName=calf_twist_01_l}}]
    constraintWeights [1,1,1]
   }
   n29 = "Constraint - base_shoes_01_l (PPP)"!(ComponentPose=@n28, Alpha=1, ConstraintWeights_0=0.9071, ConstraintWeights_1:float=0.0919, ConstraintWeights_2:float=0.0009) -> (Pose) {
    alphaBoolBlend {blendOption=Linear}
    boneToModify {boneName=base_shoes_01_l}
    constraintSetup [{targetBone {boneName=foot_l}}, {targetBone {boneName=calf_twist_01_l}}, {targetBone {boneName=ball_l}}]
    constraintWeights [1,1,1]
   }
   n30 = Anim Dynamics - Bone: base_shoes_02_l(ComponentPose=@n29, Alpha=1, ExternalForce="0,0,0") {
    angularDampingOverride=1
    angularSpringConstant=20000
    bAngularSpring=true
    bOverrideAngularDamping=true
    bUseGravityOverride=true
    gravityScale=0.35
    simulationSpace=RootRelative
    alphaBoolBlend {blendOption=Linear}
    boundBone {boneName=base_shoes_02_l}
    componentLinearAccScale=<1 1 1>
    componentLinearVelScale=<1 1 1>
    gravityOverride {z=-980}
    physicsBodyDefinitions [{
      boundBone {boneName=base_shoes_02_l}
      boxExtents=<20 30 30>
      constraintSetup {angularLimitsMax=<40 0 0>, angularLimitsMin=<-40 -40 0>, angularTarget=<1 0 0>}
      localJointOffset=<4 0 0>
     }]
   }
   n31 = Anim Dynamics - Bone: base_shoes_01_l(ComponentPose=@n30, Alpha=1, ExternalForce="0,0,0") {
    angularDampingOverride=1
    angularSpringConstant=20000
    bAngularSpring=true
    bOverrideAngularDamping=true
    bUseGravityOverride=true
    gravityScale=0.35
    simulationSpace=RootRelative
    alphaBoolBlend {blendOption=Linear}
    boundBone {boneName=base_shoes_01_l}
    componentLinearAccScale=<1 1 1>
    componentLinearVelScale=<1 1 1>
    gravityOverride {z=-980}
    physicsBodyDefinitions [{
      boundBone {boneName=base_shoes_01_l}
      boxExtents=<20 30 30>
      constraintSetup {angularLimitsMax=<40 0 0>, angularLimitsMin=<-40 -40 0>, angularTarget=<1 0 0>}
      localJointOffset=<4 0 0>
     }]
   }
   n32 = "Constraint - base_shoes_cn_01_l (PPP)"!(ComponentPose=@n31, Alpha=1, ConstraintWeights_0=0.8737, ConstraintWeights_1:float=0.1248, ConstraintWeights_2:float=0.0015) -> (Pose) {
    alphaBoolBlend {blendOption=Linear}
    boneToModify {boneName=base_shoes_cn_01_l}
    constraintSetup [{targetBone {boneName=foot_l}}, {targetBone {boneName=calf_twist_01_l}}, {targetBone {boneName=ball_l}}]
    constraintWeights [1,1,1]
   }
   n33 = "Constraint - base_shoes_cn_02_l (PPP)"!(ComponentPose=@n32, Alpha=1, ConstraintWeights_0=0.8737, ConstraintWeights_1:float=0.1248, ConstraintWeights_2:float=0.0015) -> (Pose) {
    alphaBoolBlend {blendOption=Linear}
    boneToModify {boneName=base_shoes_cn_02_l}
    constraintSetup [{targetBone {boneName=foot_l}}, {targetBone {boneName=calf_twist_01_l}}, {targetBone {boneName=ball_l}}]
    constraintWeights [1,1,1]
   }
   n34 = Anim Dynamics - Bone: dyn_shoes_cn_01_l(ComponentPose=@n33, Alpha=1, ExternalForce="0,0,0") {
    angularDampingOverride=1
    angularSpringConstant=20000
    bAngularSpring=true
    bOverrideAngularDamping=true
    bUseGravityOverride=true
    gravityScale=0.35
    simulationSpace=RootRelative
    alphaBoolBlend {blendOption=Linear}
    boundBone {boneName=dyn_shoes_cn_01_l}
    componentLinearAccScale=<1 1 1>
    componentLinearVelScale=<1 1 1>
    gravityOverride {z=-980}
    physicsBodyDefinitions [{
      boundBone {boneName=dyn_shoes_cn_01_l}
      boxExtents=<20 30 30>
      constraintSetup {angularLimitsMax=<30 15 0>, angularLimitsMin=<-30 -15 0>, angularTarget=<1 0 0>}
      localJointOffset=<4 0 0>
     }]
   }
   n35 = Anim Dynamics - Bone: dyn_shoes_cn_02_l(ComponentPose=@n34, Alpha=1, ExternalForce="0,0,0") {
    angularDampingOverride=1
    angularSpringConstant=20000
    bAngularSpring=true
    bOverrideAngularDamping=true
    bUseGravityOverride=true
    gravityScale=0.35
    simulationSpace=RootRelative
    alphaBoolBlend {blendOption=Linear}
    boundBone {boneName=dyn_shoes_cn_02_l}
    componentLinearAccScale=<1 1 1>
    componentLinearVelScale=<1 1 1>
    gravityOverride {z=-980}
    physicsBodyDefinitions [{
      boundBone {boneName=dyn_shoes_cn_02_l}
      boxExtents=<20 30 30>
      constraintSetup {angularLimitsMax=<30 15 0>, angularLimitsMin=<-30 -15 0>, angularTarget=<1 0 0>}
      localJointOffset=<4 0 0>
     }]
   }
   Save cached pose 'shoes_R_cache'(Pose=Component To Local(ComponentPose=@n35)) {cacheName=shoes_R_cache}
   n36 = RigidBody!(ComponentPose=Local To Component(LocalPose=Use cached pose 'surface_jacket_cache'()), Alpha, OverrideWorldGravity:Vector="0,0,0") -> (Pose) {
    bClampLinearTranslationLimitToRefPose=true
    bOverrideWorldGravity=true
    overridePhysicsAsset=&/F_MED_SeleneDonna_jacket_back_Physics
    simulationSpace=BaseBoneSpace
    alphaBoolBlend {blendOption=Linear}
    baseBoneRef {boneName=spine_04}
    simSpaceSettings {
     dampingAlpha=2
     maxAngularAcceleration=1000
     maxAngularVelocity=3000
     maxLinearAcceleration=4000
     maxLinearVelocity=1500
     worldAlpha=0.5
     externalAngularVelocity=<0.3 0.3 0.3>
     externalLinearDragV=<0.5 0.5 0.5>
     externalLinearVelocity=<0.3 0.3 0.3>
    }
   }
   n37 = Component To Local(ComponentPose=@n36)
   n38 = Use cached pose 'surface_jacket_cache'()
   n39 = Get isWearingBackbling()
   n40 = RigidBody(
    ComponentPose=Local To Component(LocalPose=Blend Poses by bool(BlendPose_0=@n38, BlendPose_1=@n37, bActiveValue=@n39) {blendType=Linear})
    OverrideWorldGravity="0,0,0"
   ) {
    bClampLinearTranslationLimitToRefPose=true
    bOverrideWorldGravity=true
    overridePhysicsAsset=&/F_MED_SeleneDonna_jacket_Physics
    simulationSpace=BaseBoneSpace
    alphaBoolBlend {blendOption=Linear}
    baseBoneRef {boneName=spine_04}
    simSpaceSettings {
     dampingAlpha=2
     maxAngularAcceleration=1000
     maxAngularVelocity=3000
     maxLinearAcceleration=4000
     maxLinearVelocity=1500
     worldAlpha=0.5
     externalAngularVelocity=<0.3 0.3 0.3>
     externalLinearDragV=<0.5 0.5 0.5>
     externalLinearVelocity=<0.3 0.3 0.3>
    }
   }
   n41 = "Transform (Modify) Bone - Bone: base_jacket_flapneck_01_l"!(ComponentPose=@n40, bAlphaBoolEnabled:bool=True, Rotation="-13.353,-15.179,-40.401") -> (Pose) {
    alphaInputType=Bool
    rotationMode=BMM_Additive
    rotationSpace=BCS_BoneSpace
    alphaBoolBlend {blendOption=Linear}
    boneToModify {boneName=base_jacket_flapneck_01_l}
    rotation=<rot -13.353 -15.179 -40.401>
   }
   n42 = "Transform (Modify) Bone - Bone: base_jacket_flapneck_01_r"!(ComponentPose=@n41, bAlphaBoolEnabled:bool=True, Rotation="-4.828,30.793,-26.486") -> (Pose) {
    alphaInputType=Bool
    rotationMode=BMM_Additive
    rotationSpace=BCS_BoneSpace
    alphaBoolBlend {blendOption=Linear}
    boneToModify {boneName=base_jacket_flapneck_01_r}
    rotation=<rot -4.828 30.793 -26.486>
   }
   n43 = RigidBody!(ComponentPose=@n42, Alpha) -> (Pose) {
    bClampLinearTranslationLimitToRefPose=true
    bOverrideWorldGravity=true
    overridePhysicsAsset=&/F_MED_SeleneDonna_neck_Physics
    simulationSpace=BaseBoneSpace
    alphaBoolBlend {blendOption=Linear}
    baseBoneRef {boneName=spine_05}
    componentAppliedLinearAccClamp=<0 0 0>
    overrideWorldGravity {z=980}
    simSpaceSettings {
     dampingAlpha=2
     maxAngularAcceleration=1000
     maxAngularVelocity=3000
     maxLinearAcceleration=4000
     maxLinearVelocity=1500
     worldAlpha=0.5
     externalAngularVelocity=<0.3 0.3 0.3>
     externalLinearDragV=<0.5 0.5 0.5>
     externalLinearVelocity=<0.3 0.3 0.3>
    }
   }
   n44 = "Constraint - base_zipper_l (PPPPPPPP)"!(
    ComponentPose=@n43
    Alpha=1
    ConstraintWeights_0=0.8668
    ConstraintWeights_1:float=0.0325
    ConstraintWeights_2:float=0.0313
    ConstraintWeights_3:float=0.025
    ConstraintWeights_4:float=0.0178
    ConstraintWeights_5:float=0.0123
    ConstraintWeights_6:float=0.0072
    ConstraintWeights_7:float=0.007
   ) -> (Pose) {
    alphaBoolBlend {blendOption=Linear}
    boneToModify {boneName=base_zipper_l}
    constraintSetup [
     {targetBone {boneName=dyn_jacket_l_0}}
     {targetBone {boneName=upperarm_twist_01_l}}
     {targetBone {boneName=dyn_jacket_l_1}}
     {targetBone {boneName=spine_05}}
     {targetBone {boneName=clavicle_l}}
     {targetBone {boneName=spine_04}}
     {targetBone {boneName=spine_03}}
     {targetBone {boneName=dyn_jacket_l_2}}
    ]
    constraintWeights [1,1,1,1,1,1,1,1]
   }
   n45 = Anim Dynamics - Bone: dyn_zipper_l!(SimulationSpace:AnimPhysSimSpaceType Enum, ComponentPose=@n44, Alpha=1, ExternalForce="0,0,0") -> (Pose) {
    angularDampingOverride=1
    angularSpringConstant=700
    bOverrideAngularDamping=true
    bUseGravityOverride=true
    gravityScale=0.35
    simulationSpace=RootRelative
    alphaBoolBlend {blendOption=Linear}
    alphaScaleBiasClamp {bInterpResult=true, interpSpeedDecreasing=2.5, interpSpeedIncreasing=8}
    boundBone {boneName=dyn_zipper_l}
    componentAppliedLinearAccClamp=<3000 3000 3000>
    componentLinearAccScale=<0.3 0.3 0.3>
    componentLinearVelScale=<0.3 0.3 0.3>
    gravityOverride {z=-980}
    physicsBodyDefinitions [{
      boundBone {boneName=dyn_zipper_l}
      boxExtents=<5 8 5>
      constraintSetup {angularLimitsMax=<0 5 10>, angularLimitsMin=<0 -3 -5>, angularTarget=<1 0 0>}
      localJointOffset=<1 0 0>
     }]
    relativeSpaceBone {boneName=root}
   }
   Save cached pose 'Jacket_cache'(Pose=Component To Local(ComponentPose=@n45)) {cacheName=Jacket_cache}
   n46 = "Constraint - base_jacket_r_2 (PPPPPPP)"!(
    ComponentPose=Local To Component(LocalPose=Use cached pose 'shoes_R_cache'())
    Alpha=1
    ConstraintWeights_0=0.6012
    ConstraintWeights_1:float=0.1231
    ConstraintWeights_2:float=0.1149
    ConstraintWeights_3:float=0.0906
    ConstraintWeights_4:float=0.0695
    ConstraintWeights_5:float=0.0004
    ConstraintWeights_6:float=0.0002
   ) -> (Pose) {
    alphaBoolBlend {blendOption=Linear}
    boneToModify {boneName=base_jacket_r_2}
    constraintSetup [
     {targetBone {boneName=spine_04}}
     {targetBone {boneName=clavicle_r}}
     {targetBone {boneName=spine_05}}
     {targetBone {boneName=upperarm_twist_01_r}}
     {targetBone {boneName=spine_03}}
     {targetBone {boneName=clavicle_l}}
     {targetBone {boneName=dyn_jacket_flapcn}}
    ]
    constraintWeights [1,1,1,1,1,1,1]
   }
   n47 = "Constraint - base_jacket_r_1 (PPPPP)"!(
    ComponentPose=@n46
    Alpha=1
    ConstraintWeights_0=0.5822
    ConstraintWeights_1:float=0.1873
    ConstraintWeights_2:float=0.1027
    ConstraintWeights_3:float=0.0709
    ConstraintWeights_4:float=0.0568
   ) -> (Pose) {
    alphaBoolBlend {blendOption=Linear}
    boneToModify {boneName=base_jacket_r_1}
    constraintSetup [
     {targetBone {boneName=spine_04}}
     {targetBone {boneName=upperarm_twist_01_r}}
     {targetBone {boneName=clavicle_r}}
     {targetBone {boneName=spine_03}}
     {targetBone {boneName=spine_05}}
    ]
    constraintWeights [1,1,1,1,1]
   }
   n48 = "Constraint - base_jacket_r_0 (PPPPPPP)"!(
    ComponentPose=@n47
    Alpha=1
    ConstraintWeights_0=0.8758
    ConstraintWeights_1:float=0.0384
    ConstraintWeights_2:float=0.0364
    ConstraintWeights_3:float=0.0228
    ConstraintWeights_4:float=0.0199
    ConstraintWeights_5:float=0.0034
    ConstraintWeights_6:float=0.0034
   ) -> (Pose) {
    alphaBoolBlend {blendOption=Linear}
    boneToModify {boneName=base_jacket_r_0}
    constraintSetup [
     {targetBone {boneName=spine_04}}
     {targetBone {boneName=spine_03}}
     {targetBone {boneName=upperarm_twist_01_r}}
     {targetBone {boneName=spine_05}}
     {targetBone {boneName=clavicle_r}}
     {targetBone {boneName=spine_02}}
     {targetBone {boneName=dyn_jacket_flapcn}}
    ]
    constraintWeights [1,1,1,1,1,1,1]
   }
   n49 = "Constraint - base_jacket_back (PPPPPPP)"!(
    ComponentPose=@n48
    Alpha=1
    ConstraintWeights_0=0.6665
    ConstraintWeights_1:float=0.1302
    ConstraintWeights_2:float=0.0735
    ConstraintWeights_3:float=0.0684
    ConstraintWeights_4:float=0.0606
    ConstraintWeights_5:float=0.0005
    ConstraintWeights_6:float=0.0004
   ) -> (Pose) {
    alphaBoolBlend {blendOption=Linear}
    boneToModify {boneName=base_jacket_back}
    constraintSetup [
     {targetBone {boneName=spine_04}}
     {targetBone {boneName=spine_05}}
     {targetBone {boneName=clavicle_r}}
     {targetBone {boneName=clavicle_l}}
     {targetBone {boneName=spine_03}}
     {targetBone {boneName=spine_02}}
     {targetBone {boneName=upperarm_twist_01_r}}
    ]
    constraintWeights [1,1,1,1,1,1,1]
   }
   n50 = "Constraint - base_jacket_l_2 (PPPPPP)"!(
    ComponentPose=@n49
    Alpha=1
    ConstraintWeights_0=0.6143
    ConstraintWeights_1:float=0.1031
    ConstraintWeights_2:float=0.0991
    ConstraintWeights_3:float=0.0779
    ConstraintWeights_4:float=0.0723
    ConstraintWeights_5:float=0.0332
   ) -> (Pose) {
    alphaBoolBlend {blendOption=Linear}
    boneToModify {boneName=base_jacket_l_2}
    constraintSetup [
     {targetBone {boneName=spine_04}}
     {targetBone {boneName=clavicle_l}}
     {targetBone {boneName=spine_05}}
     {targetBone {boneName=spine_03}}
     {targetBone {boneName=upperarm_twist_01_l}}
     {targetBone {boneName=clavicle_r}}
    ]
    constraintWeights [1,1,1,1,1,1]
   }
   n51 = "Constraint - base_jacket_l_1 (PPPPPP)"!(
    ComponentPose=@n50
    Alpha=1
    ConstraintWeights_0=0.5605
    ConstraintWeights_1:float=0.2039
    ConstraintWeights_2:float=0.1099
    ConstraintWeights_3:float=0.0677
    ConstraintWeights_4:float=0.057
    ConstraintWeights_5:float=0.001
   ) -> (Pose) {
    alphaBoolBlend {blendOption=Linear}
    boneToModify {boneName=base_jacket_l_1}
    constraintSetup [
     {targetBone {boneName=spine_04}}
     {targetBone {boneName=upperarm_twist_01_l}}
     {targetBone {boneName=clavicle_l}}
     {targetBone {boneName=spine_03}}
     {targetBone {boneName=spine_05}}
     {targetBone {boneName=spine_02}}
    ]
    constraintWeights [1,1,1,1,1,1]
   }
   n52 = "Constraint - base_jacket_l_0 (PPPPPPPP)"!(
    ComponentPose=@n51
    Alpha=1
    ConstraintWeights_0=0.8878
    ConstraintWeights_1:float=0.0363
    ConstraintWeights_2:float=0.0342
    ConstraintWeights_3:float=0.0193
    ConstraintWeights_4:float=0.0172
    ConstraintWeights_5:float=0.0037
    ConstraintWeights_6:float=0.001
    ConstraintWeights_7:float=0.0005
   ) -> (Pose) {
    alphaBoolBlend {blendOption=Linear}
    boneToModify {boneName=base_jacket_l_0}
    constraintSetup [
     {targetBone {boneName=spine_04}}
     {targetBone {boneName=spine_03}}
     {targetBone {boneName=upperarm_twist_01_l}}
     {targetBone {boneName=spine_05}}
     {targetBone {boneName=clavicle_l}}
     {targetBone {boneName=spine_02}}
     {targetBone {boneName=dyn_jacket_flapcn}}
     {targetBone {boneName=dyn_jacket_flapneck_01_r}}
    ]
    constraintWeights [1,1,1,1,1,1,1,1]
   }
   Save cached pose 'surface_jacket_cache'(Pose=Component To Local(ComponentPose=@n52)) {cacheName=surface_jacket_cache}
  ]}
 EventGraph {nodes [
   Event BlueprintUpdateAnimation()
   > n1 = Gravity Override State Function DEPRECATED(
    CharacterPartAnimInstance=Self-Reference()
    GameTimeInSeconds=GetGameTimeInSeconds()
    GravityOverrideParams=Get Gravity Override Params()
    RBAN Params=Get RBAN Params()
   )
   > Set GravityOverride(GravityOverride=@n1)
   > Set Sim Space Settings(Sim Space Settings:Sim Space Settings Structure=Break RBANSetting Struct!(RBANSettingStruct:"RBANSetting Struct Structure (by ref)"=@n1.RBAN SettingsOut) -> (
     ComponentLinearAccScale_28_#1:Vector=~
     ComponentLinearVelScale_31_#2:Vector=~
     SimSpaceSettings_34_#3:Sim Space Settings Structure
    ) {is=title K2Node_BreakStruct}.SimSpaceSettings_34_#3) -> (Output_Get:Sim Space Settings Structure)
   n2 = "Sin (Radians)"(A=float * float(A=GetRealTimeSeconds(), B=0.5))
   > Set Translate Star B1(Translate Star B1:Vector=MakeVector(X=float * float(A=@n2, B=0), Y=float / float(A=@n2, B=5) {is=title K2Node_CallFunction}, Z=float / float(A=@n2, B=2))) -> (Output_Get:Vector)
   n3 = "Sin (Radians)"(A=float * float(A=GetRealTimeSeconds(), B=0.4))
   > Set Translate Star B2(Translate Star B2:Vector=MakeVector(X=float * float(A=@n3, B=0.2), Y=float / float(A=@n3, B=0.4), Z=float / float(A=@n3, B=0.7))) -> (Output_Get:Vector)
   n4 = "Sin (Radians)"(A=float * float(A=GetRealTimeSeconds(), B=1.2))
   > Set Translate Star B3(Translate Star B3:Vector=MakeVector(X=float * float(A=@n4, B=0.5), Y=float / float(A=@n4, B=1))) -> (Output_Get:Vector)
   n5 = "Sin (Radians)"(A=float * float(A=GetRealTimeSeconds(), B=0.8))
   > Set Translate Star B4(Translate Star B4:Vector=MakeVector(X=float * float(A=@n5, B=0), Y=float / float(A=@n5, B=2))) -> (Output_Get:Vector)
   n6 = "Sin (Radians)"(A=float * float(A=GetRealTimeSeconds(), B=2))
   > Set Translate Star B5(Translate Star B5:Vector=MakeVector(X=float * float(A=@n6, B=0.5), Y=float / float(A=@n6, B=1), Z=float / float(A=@n6, B=3))) -> (Output_Get:Vector)
   n7 = "Sin (Radians)"(A=float * float(A=GetRealTimeSeconds(), B=1))
   > Set Translate Star B6(Translate Star B6:Vector=MakeVector(X=float * float(A=@n7, B=0.5), Y=float / float(A=@n7, B=1))) -> (Output_Get:Vector)
   n8 = "Sin (Radians)"(A=float * float(A=GetRealTimeSeconds(), B=2.8))
   > Set Translate Star B7(Translate Star B7:Vector=MakeVector(X=float * float(A=@n8, B=0.5), Y=float / float(A=@n8, B=1))) -> (Output_Get:Vector)
   n9 = "Sin (Radians)"(A=float * float(A=GetRealTimeSeconds(), B=2))
   > Set Translate Star B8(Translate Star B8:Vector=MakeVector(X=float * float(A=@n9, B=0.5), Y=float / float(A=@n9, B=3))) -> (Output_Get:Vector)
   n10 = CheckForBackbling()
   > Set isWearingBackbling(isWearingBackbling=AND Boolean!(A=@n10, B=NOT Boolean(A=@n10.bIsBackblingACape), C:bool=NOT Boolean(A=@n10.bIsBackblingLong)) -> (ReturnValue))
   > Set Is Wearing Cape(Is Wearing Cape=OR Boolean(A=@n10.bIsBackblingACape, B=@n10.bIsBackblingLong))
   n11 = Get Zipper() -> (Zipper:bool)
   > Set Zipper(Zipper:bool=@n11) -> (Output_Get:bool)
   n12 = Get bIsInFrontEnd()
   > n13 = Branch(Condition=@n12)
   > Set alpha frontend(alpha frontend:double=0.5) -> (Output_Get:double)
   > Set simspace var(simspace var:Byte=2) -> (Output_Get:Byte=0)
   Set alpha frontend(execute=@n13.else, alpha frontend=1)
   > Set simspace var(simspace var=3) -> (Output_Get=0)
   float / float(A=0, B=0)
   float / float(A=0, B=0)
   float / float(A=0, B=0)
   NOT Boolean(A=@n12)
   float / float(A=0, B=0)
   float / float(A=0, B=0)
   Set Zipper(Zipper=@n11)
  ]}
]
variables [
 Alpha Frontend {type=number, default=1}
 Gravity Override Params {type=GravityOverrideParamsStruct}
 GravityOverride {type=Vector}
 Is Wearing Cape {type=boolean}
 isWearingBackbling {type=boolean}
 RBAN Params {
  type=RBANParamsStruct
  default {
   dBNO {simSpaceSettings {
     maxAngularAcceleration=1000
     maxAngularVelocity=3
     maxLinearAcceleration=300
     maxLinearVelocity=700
     velocityScaleZ=0
     externalLinearDragV=<0 0.5 0>
    }}
   dBNO_Carried {simSpaceSettings {
     maxAngularAcceleration=1000
     maxAngularVelocity=3
     maxLinearAcceleration=300
     maxLinearVelocity=700
     velocityScaleZ=0
     externalLinearDragV=<0 0.5 0>
    }}
   emoteOrMelee {simSpaceSettings {
     maxAngularAcceleration=1000
     maxAngularVelocity=3
     maxLinearAcceleration=300
     maxLinearVelocity=700
     velocityScaleZ=0
     externalLinearDragV=<0 0.5 0>
    }}
   front_End {simSpaceSettings {
     maxAngularAcceleration=1000
     maxAngularVelocity=3
     maxLinearAcceleration=300
     maxLinearVelocity=700
     velocityScaleZ=0
     externalLinearDragV=<0 0.5 0>
    }}
   onGround_Crouch_Moving {simSpaceSettings {
     maxAngularAcceleration=1000
     maxAngularVelocity=3
     maxLinearAcceleration=300
     maxLinearVelocity=700
     velocityScaleZ=0
     externalLinearDragV=<0 0.5 0>
    }}
   onGround_Falling {simSpaceSettings {
     maxAngularAcceleration=1000
     maxAngularVelocity=3
     maxLinearAcceleration=300
     maxLinearVelocity=700
     velocityScaleZ=0
     externalLinearDragV=<0 0.5 0>
    }}
   onGround_Moving {simSpaceSettings {
     maxAngularAcceleration=1000
     maxAngularVelocity=3
     maxLinearAcceleration=300
     maxLinearVelocity=700
     velocityScaleZ=0
     externalLinearDragV=<0 0.5 0>
    }}
   onGround_Standing {simSpaceSettings {
     maxAngularAcceleration=1000
     maxAngularVelocity=3
     maxLinearAcceleration=300
     maxLinearVelocity=700
     velocityScaleZ=0
     externalLinearDragV=<0 0.5 0>
    }}
   rail_Grinding {simSpaceSettings {
     maxAngularAcceleration=1000
     maxAngularVelocity=3
     maxLinearAcceleration=300
     maxLinearVelocity=700
     velocityScaleZ=0
     worldAlpha=0.1
     externalLinearDragV=<0 0.5 0>
    }}
   riding_Moving {simSpaceSettings {
     maxAngularAcceleration=1000
     maxAngularVelocity=3
     maxLinearAcceleration=300
     maxLinearVelocity=700
     velocityScaleZ=0
     externalLinearDragV=<0 0.5 0>
    }}
   riding_Moving_Targeting {simSpaceSettings {
     maxAngularAcceleration=1000
     maxAngularVelocity=3
     maxLinearAcceleration=300
     maxLinearVelocity=700
     velocityScaleZ=0
     externalLinearDragV=<0 0.5 0>
    }}
   riding_Standing {simSpaceSettings {
     maxAngularAcceleration=1000
     maxAngularVelocity=3
     maxLinearAcceleration=300
     maxLinearVelocity=700
     velocityScaleZ=0
     externalLinearDragV=<0 0.5 0>
    }}
   skydive_DivingDown {simSpaceSettings {
     maxAngularAcceleration=1000
     maxAngularVelocity=3
     maxLinearAcceleration=300
     maxLinearVelocity=700
     velocityScaleZ=0
     externalLinearDragV=<0 0.5 0>
    }}
   skydive_FloatingDown {
    componentLinearAccScale=<0 0.5 0>
    componentLinearVelScale=<0 0.5 0>
    simSpaceSettings {
     maxAngularAcceleration=1000
     maxAngularVelocity=3
     maxLinearAcceleration=300
     maxLinearVelocity=700
     velocityScaleZ=0
     externalLinearDragV=<0 0.5 0>
    }
   }
   skydive_GoingUp {simSpaceSettings {
     maxAngularAcceleration=1000
     maxAngularVelocity=3
     maxLinearAcceleration=300
     maxLinearVelocity=700
     velocityScaleZ=0
     externalLinearDragV=<0 0.5 0>
    }}
   skydive_Parachuting {
    componentLinearAccScale=<0 0.1 0.3>
    componentLinearVelScale=<0 0.1 0.3>
    simSpaceSettings {
     maxAngularAcceleration=1000
     maxAngularVelocity=3
     maxLinearAcceleration=300
     maxLinearVelocity=700
     velocityScaleZ=0
     externalLinearDragV=<0 0.5 0>
    }
   }
   skydive_Surfing {simSpaceSettings {
     maxAngularAcceleration=1000
     maxAngularVelocity=3
     maxLinearAcceleration=300
     maxLinearVelocity=700
     velocityScaleZ=0
     externalLinearDragV=<0 0.5 0>
    }}
   swimming {simSpaceSettings {
     maxAngularAcceleration=1000
     maxAngularVelocity=3
     maxLinearAcceleration=300
     maxLinearVelocity=700
     velocityScaleZ=0
     externalLinearDragV=<0 0.5 0>
    }}
   tactical_Sprint {simSpaceSettings {
     maxAngularAcceleration=1000
     maxAngularVelocity=3
     maxLinearAcceleration=300
     maxLinearVelocity=700
     velocityScaleZ=0
     externalLinearDragV=<0 0.5 0>
    }}
   ziplining {simSpaceSettings {
     maxAngularAcceleration=1000
     maxAngularVelocity=3
     maxLinearAcceleration=300
     maxLinearVelocity=700
     velocityScaleZ=0
     externalLinearDragV=<0 0.5 0>
    }}
  }
 }
 Sim Space Settings {
  type=SimSpaceSettings
  default {
   dampingAlpha=1
   maxAngularAcceleration=1000
   maxAngularVelocity=1000
   maxLinearAcceleration=1000
   maxLinearVelocity=1000
   velocityScaleZ=1
   worldAlpha=0
   externalLinearDragV=<0 0 0>
  }
 }
 Simspace Var {type=integer, default=2}
 Translate Star B1 {type=Vector}
 Translate Star B2 {type=Vector}
 Translate Star B3 {type=Vector}
 Translate Star B4 {type=Vector}
 Translate Star B5 {type=Vector}
 Translate Star B6 {type=Vector}
 Translate Star B7 {type=Vector}
 Translate Star B8 {type=Vector}
 Zipper {type=boolean}
]
