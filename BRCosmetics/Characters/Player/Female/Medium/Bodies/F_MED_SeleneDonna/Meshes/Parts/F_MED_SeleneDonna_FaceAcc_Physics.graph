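# Physics asset for the SeleneDonna FaceAcc mesh, setting up secondary motion for hair buns, front strands, fringe and two star chains.
# Capsule chains on dyn_ bones hang from base_ bones, mostly locked to the head sphere, with tight swing limits on hair and wider ones on stars.
# The base_star_b_02 anchor is angularly free against the head.
# Large capsules on the right upper arm form a separate locked group with the clavicle.
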PhysicsAsset /BRCosmetics/Characters/Player/Female/Medium/Bodies/F_MED_SeleneDonna/Meshes/Parts/F_MED_SeleneDonna_FaceAcc_Physics
bodies {
 base_fr_01_l [{shapeType=Capsule, length=1, radius=1, center=<2 0 0>, rotation=<rot 90 0 0>}]
 base_fr_a_01_r [{shapeType=Capsule, length=0.505, radius=0.505, shapeName=base_fr_a_01_r_capsule, rotation=<rot 0 -90 0>}]
 base_fr_b_01_r [{shapeType=Capsule, length=0.505, radius=0.505, shapeName=base_fr_b_01_r_capsule, rotation=<rot 0 -90 0>}]
 base_fr_fringe_01 [{shapeType=Capsule, length=0.505, radius=0.505, shapeName=base_fr_fringe_01_capsule, rotation=<rot 0 -90 0>}]
 base_star_b_02 [{shapeType=Capsule, length=0.505, radius=0.505, shapeName=base_star_b_02_capsule, rotation=<rot 0 -90 0>}]
 base_stars [{shapeType=Capsule, length=0.505, radius=0.1, shapeName=base_stars_capsule, rotation=<rot 0 -90 0>}]
 clavicle_r [{shapeType=Capsule, length=0.505, radius=0.505, shapeName=clavicle_r_capsule, rotation=<rot 0 -90 0>}]
 dyn_bun_l [{shapeType=Capsule, length=0.5, radius=6.4263, shapeName=dyn_bun_l_capsule, center=<5.4943 0 0>, rotation=<rot 90 180 0>}]
 dyn_bun_r [{shapeType=Capsule, length=0.5, radius=6.4263, shapeName=dyn_bun_r_capsule, center=<-5.4943 0 0>, rotation=<rot 90 0 0>}]
 dyn_fr_01_l [{shapeType=Capsule, length=1, radius=1, center=<2 0 0>, rotation=<rot 90 0 0>}]
 dyn_fr_02_l [{shapeType=Capsule, length=1, radius=1, center=<2 0 0>, rotation=<rot 90 0 0>}]
 dyn_fr_03_l [{shapeType=Capsule, length=1, radius=1, center=<2 -1.2 0>, rotation=<rot 90 0 0>}]
 dyn_fr_a_01_r [{shapeType=Capsule, length=1, radius=0.8, shapeName=dyn_fr_a_01_r_capsule, center=<1.1793 0 0>, rotation=<rot 90 0 0>}]
 dyn_fr_a_02_r [{shapeType=Capsule, length=1, radius=0.8, shapeName=dyn_fr_a_02_r_capsule, center=<1.1793 0 0>, rotation=<rot 90 0 0>}]
 dyn_fr_a_03_r [{shapeType=Capsule, length=0.8, radius=0.6, shapeName=dyn_fr_a_03_r_capsule, center=<1.1793 0 0>, rotation=<rot 90 0 0>}]
 dyn_fr_b_01_r [{shapeType=Capsule, length=1, radius=0.8, shapeName=dyn_fr_b_01_r_capsule, center=<1.1793 0 0>, rotation=<rot 90 0 0>}]
 dyn_fr_b_02_r [{shapeType=Capsule, length=1, radius=0.8, shapeName=dyn_fr_b_02_r_capsule, center=<1.1793 0 0>, rotation=<rot 90 0 0>}]
 dyn_fr_b_03_r [{shapeType=Capsule, length=1, radius=0.8, shapeName=dyn_fr_b_03_r_capsule, center=<1.1793 0.14 0>, rotation=<rot 90 0 0>}]
 dyn_fr_fringe_01 [{
   shapeType=Capsule
   length=1.6386
   radius=0.7355
   shapeName=dyn_fr_fringe_01_capsule
   center=<5.8791 -1.3785 0.8403>
   rotation=<rot -5.8888 -90 12.4022>
  }]
 dyn_fr_fringe_02 [{shapeType=Capsule, length=1, radius=1, shapeName=dyn_fr_fringe_02_capsule, center=<4.0347 0 0>, rotation=<rot 90 0 0>}]
 dyn_fr_fringe_03 [{shapeType=Capsule, length=2.5, radius=1.3, shapeName=dyn_fr_fringe_03_capsule, center=<3.3644 0 0>, rotation=<rot 90 0 0>}]
 dyn_fr_fringe_04 [{shapeType=Capsule, length=0.5, radius=1.8, shapeName=dyn_fr_fringe_04_capsule, center=<2.7831 0 0>, rotation=<rot 90 0 0>}]
 dyn_fr_fringe_05 [{shapeType=Capsule, length=0.5, radius=2.8, shapeName=dyn_fr_fringe_05_capsule, center=<3.5125 0 0>, rotation=<rot 90 0 0>}]
 dyn_star_a_01 [{shapeType=Capsule, length=1.5, radius=0.3, shapeName=dyn_star_a_01_capsule, center=<2.6277 0 0>, rotation=<rot 90 0 0>}]
 dyn_star_a_02 [{shapeType=Capsule, length=3, radius=0.5, shapeName=dyn_star_a_02_capsule, center=<1.9016 0 0>, rotation=<rot 90 0 0>}]
 dyn_star_a_03 [
  {shapeType=Capsule, length=4.5, radius=1, shapeName=dyn_star_a_02_capsule, center=<2.1351 -0.2608 0>}
  {shapeType=Capsule, length=2.5, radius=1.5, shapeName=dyn_star_a_03_capsule, center=<5.4612 -0.3638 0>}
 ]
 dyn_star_b_01 [{shapeType=Capsule, length=1.5, radius=0.3, shapeName=dyn_star_b_01_capsule, center=<2.7987 0 0>, rotation=<rot 90 0 0>}]
 dyn_star_b_02 [{shapeType=Capsule, length=3, radius=0.5, shapeName=dyn_star_b_02_capsule, center=<1.9164 0 0>, rotation=<rot 90 0 0>}]
 dyn_star_b_03 [
  {
   shapeType=Capsule
   length=6.6417
   radius=0.6583
   shapeName=dyn_star_b_02_capsule
   center=<2.7676 -0.3315 0>
   rotation=<rot 0.0002 0 -26.5657>
  }
  {shapeType=Capsule, length=4, radius=1, shapeName=dyn_star_b_03_capsule, center=<4.895 -0.4042 -0.1297>, rotation=<rot 0.0001 0 -17.7885>}
 ]
 head [{shapeType=Sphere, radius=13.5, shapeName=head_sphere, center=<5.599 1.0404 -0.5054>}]
 helper_upperarm_r [{shapeType=Capsule, length=12.5, radius=10.7, shapeName=helper_upperarm_r_capsule, center=<-13.9087 0 3.9239>, rotation=<rot 90 0 0>}]
 upperarm_r [{shapeType=Capsule, length=20, radius=10, shapeName=upperarm_r_capsule, center=<-11.9481 -0.9047 4.112>, rotation=<rot 90 0 0>}]
}
constraints [
 {
  bone1Name=base_stars
  bone2Name=head
  swing1LimitDegrees=45
  swing1Motion=Locked
  swing2LimitDegrees=45
  swing2Motion=Locked
  twistLimitDegrees=45
  twistMotion=Locked
 }
 {bone1Name=dyn_star_b_01, bone2Name=base_star_b_02, swing1LimitDegrees=10, swing2LimitDegrees=50, twistLimitDegrees=0}
 {bone1Name=dyn_star_b_02, bone2Name=dyn_star_b_01, swing1LimitDegrees=20, swing2LimitDegrees=15, twistLimitDegrees=0}
 {bone1Name=dyn_star_a_02, bone2Name=dyn_star_a_01, swing1LimitDegrees=25, swing2LimitDegrees=25, twistLimitDegrees=0}
 {bone1Name=dyn_star_b_03, bone2Name=dyn_star_b_02, swing1LimitDegrees=25, swing2LimitDegrees=30, twistLimitDegrees=10}
 {bone1Name=dyn_star_a_03, bone2Name=dyn_star_a_02, swing1LimitDegrees=20, swing2LimitDegrees=25, twistLimitDegrees=10}
 {bone1Name=dyn_bun_l, bone2Name=head, swing1LimitDegrees=3, swing2LimitDegrees=10, twistLimitDegrees=5}
 {bone1Name=dyn_bun_r, bone2Name=head, swing1LimitDegrees=3, swing2LimitDegrees=5, twistLimitDegrees=5}
 {bone1Name=dyn_fr_b_03_r, bone2Name=dyn_fr_b_02_r, swing1LimitDegrees=0, swing2LimitDegrees=4, twistLimitDegrees=0}
 {bone1Name=dyn_fr_b_02_r, bone2Name=dyn_fr_b_01_r, swing1LimitDegrees=0, swing2LimitDegrees=3, twistLimitDegrees=0}
 {bone1Name=dyn_fr_b_01_r, bone2Name=base_fr_b_01_r, swing1LimitDegrees=0, swing2LimitDegrees=2, twistLimitDegrees=0}
 {bone1Name=dyn_fr_a_03_r, bone2Name=dyn_fr_a_02_r, swing1LimitDegrees=0, swing2LimitDegrees=4, twistLimitDegrees=0, twistMotion=Locked}
 {bone1Name=dyn_fr_a_02_r, bone2Name=dyn_fr_a_01_r, swing1LimitDegrees=0, swing2LimitDegrees=3, twistLimitDegrees=0}
 {bone1Name=dyn_fr_a_01_r, bone2Name=base_fr_a_01_r, swing1LimitDegrees=0, swing2LimitDegrees=3, twistLimitDegrees=0}
 {bone1Name=dyn_fr_03_l, bone2Name=dyn_fr_02_l, swing1LimitDegrees=2, swing2LimitDegrees=2, twistLimitDegrees=0}
 {bone1Name=dyn_fr_02_l, bone2Name=dyn_fr_01_l, swing1LimitDegrees=2, swing2LimitDegrees=2, twistLimitDegrees=0}
 {bone1Name=dyn_fr_01_l, bone2Name=base_fr_01_l, swing1LimitDegrees=2, swing2LimitDegrees=0, twistLimitDegrees=0}
 {
  bone1Name=base_fr_b_01_r
  bone2Name=head
  swing1LimitDegrees=45
  swing1Motion=Locked
  swing2LimitDegrees=45
  swing2Motion=Locked
  twistLimitDegrees=45
  twistMotion=Locked
 }
 {
  bone1Name=base_fr_a_01_r
  bone2Name=head
  swing1LimitDegrees=45
  swing1Motion=Locked
  swing2LimitDegrees=45
  swing2Motion=Locked
  twistLimitDegrees=45
  twistMotion=Locked
 }
 {
  bone1Name=base_fr_01_l
  bone2Name=head
  swing1LimitDegrees=45
  swing1Motion=Locked
  swing2LimitDegrees=45
  swing2Motion=Locked
  twistLimitDegrees=45
  twistMotion=Locked
 }
 {bone1Name=dyn_fr_01_l, bone2Name=head, swing1LimitDegrees=2, swing2LimitDegrees=2, twistLimitDegrees=0}
 {
  bone1Name=helper_upperarm_r
  bone2Name=upperarm_r
  swing1LimitDegrees=45
  swing1Motion=Locked
  swing2LimitDegrees=45
  swing2Motion=Locked
  twistLimitDegrees=45
  twistMotion=Locked
 }
 {
  bone1Name=upperarm_r
  bone2Name=clavicle_r
  swing1LimitDegrees=45
  swing1Motion=Locked
  swing2LimitDegrees=45
  swing2Motion=Locked
  twistLimitDegrees=45
  twistMotion=Locked
 }
 {
  bone1Name=base_fr_fringe_01
  bone2Name=head
  swing1LimitDegrees=45
  swing1Motion=Locked
  swing2LimitDegrees=45
  swing2Motion=Locked
  twistLimitDegrees=45
  twistMotion=Locked
 }
 {
  bone1Name=dyn_fr_fringe_01
  bone2Name=base_fr_fringe_01
  swing1LimitDegrees=45
  swing1Motion=Locked
  swing2LimitDegrees=45
  swing2Motion=Locked
  twistLimitDegrees=45
  twistMotion=Locked
 }
 {bone1Name=dyn_fr_fringe_02, bone2Name=dyn_fr_fringe_01, swing1LimitDegrees=2, swing2LimitDegrees=2, twistLimitDegrees=0}
 {bone1Name=dyn_fr_fringe_03, bone2Name=dyn_fr_fringe_02, swing1LimitDegrees=2, swing2LimitDegrees=2, twistLimitDegrees=0}
 {bone1Name=dyn_fr_fringe_04, bone2Name=dyn_fr_fringe_03, swing1LimitDegrees=1, swing2LimitDegrees=2, twistLimitDegrees=0}
 {bone1Name=dyn_fr_fringe_05, bone2Name=dyn_fr_fringe_04, swing1LimitDegrees=2, swing2LimitDegrees=2, twistLimitDegrees=0}
 {bone1Name=dyn_star_a_01, bone2Name=base_stars, swing1LimitDegrees=15, swing2LimitDegrees=40, twistLimitDegrees=0}
 {
  bone1Name=base_star_b_02
  bone2Name=head
  swing1LimitDegrees=45
  swing1Motion=Free
  swing2LimitDegrees=45
  swing2Motion=Free
  twistLimitDegrees=45
  twistMotion=Free
 }
]
